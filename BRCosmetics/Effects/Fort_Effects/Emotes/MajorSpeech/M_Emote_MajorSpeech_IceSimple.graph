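# Translucent lit material for a simple ice look, used on Niagara mesh particles in the MajorSpeech emote effects.
# A tinted nebula texture at half UV tiling is added to a colored Fresnel rim with parameterized exponent and base reflect fraction.
# Particle color tints only the Fresnel term, and the sum feeds the shared MF_FX_EmissiveAdjustments function as emissive.
# A separate BaseColor parameter supplies a blue color.
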Material /BRCosmetics/Effects/Fort_Effects/Emotes/MajorSpeech/M_Emote_MajorSpeech_IceSimple
graph {
 properties {bAutomaticallySetUsageInEditor=false, bUsedWithNiagaraMeshParticles=true, blendMode=BLEND_Translucent, usedShadingModels=Default Lit}
 nodes [
  n1 = VectorParameter() {parameterName=MainTex_Color, defaultValue=<rgba 0.2526 0.8825 1.375 1>}
  n2 = TextureSample(UVs=TextureCoordinate() {uTiling=0.5, vTiling=0.5}) {texture=/Game/Characters/Player/Female/Medium/Bodies/F_MED_Celestial_Glow/Textures/T_FN_Nebula_LighterBlue}
  n3 = Multiply(
   A=VectorParameter() {parameterName=Fresnel_Color, defaultValue=<rgba 0.5208 3.4457 5 1>}.RGB
   B=Fresnel(
    ExponentIn=ScalarParameter() {defaultValue=2, parameterName=Fresnel_Exp}
    BaseReflectFractionIn=ScalarParameter() {parameterName=Fresnel_BRF}
   )
  )
  n4 = ParticleColor()
  MaterialFunctionCall(Emissive=Add(A=Multiply(A=@n1.RGB, B=@n2.RGB), B=Multiply(A=@n3, B=@n4.RGB))) {materialFunction=&MF_FX_EmissiveAdjustments}
  VectorParameter() {parameterName=BaseColor, defaultValue=<rgba 0.0495 0.3259 0.5 1>}
 ]
}
