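# Translucent unlit Niagara mesh-particle material for glowing, starry smoke in the MajorSpeech emote.
# Emissive layers world-aligned screen-space dust, nebula and clouds with fast-panning star textures, Fresnel gradient and normal-shaded noise, partly tinted by particle color.
# Opacity combines particle alpha, noise-modulated view-angle fade, sine UV mask, depth fade and optional erosion dissolve driven by a dynamic parameter.
# Panning noise, scaled by another dynamic parameter, displaces vertices along their normals.
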Material /BRCosmetics/Effects/Fort_Effects/Emotes/MajorSpeech/M_Emote_MajorSpeech_MeshSmoke
graph {
 properties {
  bUsedWithNiagaraMeshParticles=true
  blendMode=BLEND_Translucent
  opacityMaskClipValue=0.1
  pixelDepthOffsetMode=PDOM_Legacy
  shadingModel=MSM_Unlit
  usedShadingModels=Unlit
 }
 nodes [
  n1 = ParticleColor()
  n2 = ScalarParameter() {defaultValue=10, group=Opacity, parameterName=Edge Fade Sharpness}
  n3 = TextureCoordinate()
  n4 = AppendVector(A=Constant(), B=ParticleRandom())
  n5 = TextureSample(UVs=Panner(Coordinate=Add(A=@n3, B=@n4)) {speedY=0.7}) {texture=&T_Inky_Smoke_Tile}
  n6 = Power(
   Base=MaterialFunctionCall("FadeSensitivity (optional)"=ScalarParameter() {defaultValue=2, group=Opacity, parameterName=Edge Fade}) {materialFunction=&FadeBasedOnViewAngle}.Result
   Exp=LinearInterpolate(Alpha=@n5.R) {constA=3, constB=10}
  )
  n7 = Saturate(None=MaterialFunctionCall!(
    Texture=@n5.R
    Erosion=DynamicParameter() {defaultValue=<rgba 1 1 1 0.1>, paramNames [WPO Intensity, BrightOffset, LightningFlash, Erode]}.Erode
    Softness=ScalarParameter() {defaultValue=0.5, group=Opacity, parameterName=Param_2}
   ) {materialFunction=/Game/Effects/Niagara/MasterMaterials/Functions/MF_Erosion_Smoothstep_01}.Result)
  n8 = Multiply(
   A=Saturate(None=Multiply(A=@n2, B=@n6) {constA=10})
   B=LinearInterpolate(A=Constant() {r=1}, B=OneMinus(None=@n7), Alpha=ScalarParameter() {group=Opacity, parameterName=UseDissolve?})
  )
  n9 = Sine(None=ComponentMask(None=TextureCoordinate()) {g=true}) {period=2}
  n10 = ScalarParameter() {defaultValue=1.7, group=Opacity, parameterName=LinearMask_power}
  n11 = Saturate(None=Multiply(A=Power(Base=@n9, Exp=@n10) {constExponent=1.7}, B=ScalarParameter() {defaultValue=5, group=Opacity, parameterName=LinearMask_Multiply}) {constB=5})
  DepthFade(
   Opacity=Multiply(A=Multiply(A=@n1.A, B=@n8), B=LinearInterpolate(Alpha=@n11)) {constA=1}
   FadeDistance=ScalarParameter() {defaultValue=10, group=Opacity, parameterName=DepthFade}
  )
  n12 = Constant3Vector() {constant=<rgba 0.3674 1.2837 2 1>}
  n13 = Multiply(A=MaterialFunctionCall!(Center Point World Position) {materialFunction=/Engine/Functions/Engine_MaterialFunctions02/UVs/WorldSpaceAlignedScreenCoordinates}.Offset and Scaled UVS) {constB=6}
  n14 = Panner(Coordinate=@n13) {speedY=0.3}
  n15 = LinearInterpolate(
   A=Power(Base=TextureSample(UVs=@n14) {texture=/Game/Global/Textures/Default/PointSplatter/T_Tiling_Dust}.RGB) {constExponent=1.2}
   B=TextureSample(UVs=Panner(Coordinate=@n13) {speedY=0.3}) {texture=/Game/Characters/Player/Female/Medium/Bodies/F_MED_Celestial_Glow/Textures/T_FN_Nebula_LighterBlue}.RGB
  ) {constAlpha=0.3}
  n16 = TextureSample(UVs=Panner(Coordinate=@n13) {speedY=0.3}) {samplerType=SAMPLERTYPE_LinearGrayscale, texture=/Game/Athena/Prototype/Terrain/T_CloudCoverage}
  n17 = Power(Base=Multiply(A=Multiply(A=@n15, B=@n16.R)) {constB=2}) {constExponent=1.5}
  n18 = TextureCoordinate() {uTiling=2, vTiling=2.2}
  n19 = TextureCoordinate() {vTiling=1.2}
  n20 = Multiply(
   A=TextureSample(UVs=Panner(Coordinate=@n18) {speedX=0.1, speedY=4}) {texture=&/T_Emote_MajorSpeech_Star02}.R
   B=TextureSample(UVs=Panner(Coordinate=@n19) {speedX=0.1, speedY=4}) {samplerType=SAMPLERTYPE_LinearGrayscale, texture=&T_CloudCoverage}.R
  )
  n21 = Panner(Coordinate=TextureCoordinate() {uTiling=5, vTiling=5}) {speedX=0.2, speedY=6}
  n22 = DotProduct(
   A=Constant3Vector() {constant=<rgba 0 0 1 0>}
   B=TextureSample(UVs=TextureCoordinate() {vTiling=0.7}) {samplerType=SAMPLERTYPE_Normal, texture=/Game/Global/Textures/Default/Tileable_Patterns/1Normal}.RGB
  )
  n23 = VectorParameter() {group=Noise, parameterName=ColorNoise, defaultValue=<rgba 0.3549 0.7957 1 1>}
  n24 = TextureCoordinate()
  n25 = AppendVector(
   A=ScalarParameter() {defaultValue=1, group=Noise, parameterName=NoiseTextureU}
   B=ScalarParameter() {defaultValue=1, group=Noise, parameterName=NoiseTextureV}
  )
  n26 = Add(A=Add(A=Multiply(A=@n24, B=@n25), B=Constant2Vector() {g=0.5, r=0.5}), B=AppendVector(A=Constant(), B=ParticleRandom()))
  n27 = AppendVector(
   A=ScalarParameter() {group=Noise, parameterName=NoiseTextureSpeed01_X}
   B=ScalarParameter() {defaultValue=0.7, group=Noise, parameterName=NoiseTextureSpeed01_Y}
  )
  n28 = Saturate(None=TextureSampleParameter2D(UVs=Panner(Coordinate=@n26, Speed=@n27) {speedY=0.7}) {group=Noise, parameterName=NoiseTexture, texture=&T_Inky_Smoke_Tile}.R)
  n29 = Add(A=Multiply(A=ParticleColor().RGB) {constB=0.01}, B=Multiply(A=Power(Base=@n22) {constExponent=5}, B=Multiply(A=@n23.RGB, B=@n28)) {constA=1})
  n30 = LinearInterpolate(
   A=VectorParameter() {group=Color, parameterName=ColorCenter, defaultValue=<rgba 0.008 0.0383 0.08 1>}.RGB
   B=VectorParameter() {group=Color, parameterName=ColorOutside, defaultValue=<rgba 0.1038 0.5127 0.7969 1>}.RGB
   Alpha=MaterialFunctionCall(Power=Constant() {r=1.5}) {materialFunction=&Fresnel_Function}.Result
  )
  n31 = Add(
   A=Multiply(
    A=Constant3Vector() {constant=<rgba 0.2812 2.1756 3 1>}
    B=TextureSample(UVs=@n21) {samplerType=SAMPLERTYPE_Masks, texture=&/T_Emote_MajorSpeech_Star01}.G
   )
   B=Multiply(A=Add(A=@n29, B=@n30), B=@n1.RGB)
  )
  MaterialFunctionCall(Emissive=Add(A=Multiply(A=@n12, B=@n17), B=Add(A=@n20, B=@n31))) {materialFunction=&MF_FX_EmissiveAdjustments}
  n32 = TextureCoordinate()
  n33 = ParticleRandom()
  n34 = TextureSample(UVs=Panner(Coordinate=Add(A=@n32, B=@n33)) {speedX=0.5, speedY=1}) {samplerType=SAMPLERTYPE_Grayscale, texture=&LowResBlurredNoise}
  n35 = Multiply(
   A=ScalarParameter() {defaultValue=15, parameterName=WPO Intensity}
   B=DynamicParameter() {defaultValue=<rgba 1 1 1 0.1>, paramNames [WPO Intensity, BrightOffset, LightningFlash, Erode]}.WPO Intensity
  )
  NamedRerouteDeclaration(None=Multiply(A=Multiply(A=@n34.R, B=@n35) {constB=25}, B=VertexNormalWS()) {constB=0}) {name=WPO, nodeColor=<rgba 1 0 0.7734 1>}
  NamedRerouteUsage()
 ]
}
