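# Translucent unlit Niagara mesh/ribbon material for the icy, snowy icicle effect in the MajorSpeech emote.
# Emissive layers panning noise, nebula, star sparkles and optional snowflakes, one stepped via floored time; Use_TopUV blends alternate tilings.
# Opacity combines ramp and noise masks, particle-driven HeightLerp and optional top dissolves, view-angle and depth fades.
# A shared Speed reroute selects game time or particle-driven time; noise along normals builds world position offset.
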Material /BRCosmetics/Effects/Fort_Effects/Emotes/MajorSpeech/M_Emote_MajorSpeech_icicle
graph {
 properties {
  bAutomaticallySetUsageInEditor=false
  bUsedWithNiagaraMeshParticles=true
  bUsedWithNiagaraRibbons=true
  blendMode=BLEND_Translucent
  shadingModel=MSM_Unlit
  usedShadingModels=Unlit
 }
 nodes [
  n1 = ParticleColor()
  n2 = Multiply(A=TextureCoordinate() {vTiling=1.3}, B=AppendVector(A=Constant() {r=1}, B=Constant() {r=0.5}))
  n3 = TextureCoordinate() {vTiling=0.3}
  n4 = ScalarParameter() {parameterName=Use_TopUV}
  n5 = TextureSample(UVs=Panner(Coordinate=LinearInterpolate(A=@n2, B=@n3, Alpha=@n4), Time=NamedRerouteUsage()) {speedX=0.1, speedY=0.9}) {samplerType=SAMPLERTYPE_LinearGrayscale, texture=/Game/Effects/Fort_Effects/Emotes/UltraLight/Texture/T_Cloud_Greyscale}
  n6 = LinearInterpolate(Alpha=@n5.R) {constA=1, constB=1.5}
  n7 = TextureSample(UVs=Panner(Coordinate=TextureCoordinate() {uTiling=4}, Time=NamedRerouteUsage(), Speed=Constant2Vector() {g=1}) {speedY=1}) {samplerType=SAMPLERTYPE_LinearColor, texture=&T_NoisePack_LinearColor}
  n8 = Add(A=TextureCoordinate() {uTiling=0}, B=Multiply(A=LinearInterpolate(Alpha=@n7.R) {constA=0.3, constB=0.5}) {constB=0.1})
  n9 = Constant2Vector()
  n10 = Multiply(A=TextureSample(UVs=Add(A=@n8, B=@n9) {constB=0}) {samplerSource=SSM_Clamp_WorldGroupSettings, texture=/Game/Effects/Fort_Effects/Textures/Ramps/T_Power_Source_Sphere_Ramp}.R) {constB=3}
  n11 = TextureSample(UVs=Panner(Coordinate=TextureCoordinate() {uTiling=3}, Time=NamedRerouteUsage(), Speed=Constant2Vector() {g=1.5}) {speedY=1}) {samplerType=SAMPLERTYPE_LinearGrayscale, texture=&T_SimplePerlin_Greyscale}
  n12 = ComponentMask(None=TextureCoordinate()) {g=true}
  n13 = Multiply(A=Multiply(A=@n12, B=OneMinus(None=@n12))) {constB=5}
  n14 = Add(A=Multiply(A=Power(Base=@n10) {constExponent=1}, B=Multiply(A=@n11.RGB)), B=Saturate(None=Power(Base=@n13) {constExponent=4}))
  n15 = VectorParameter() {parameterName=DissolveTex_UV, defaultValue=<rgba 2 0.8 1.3 0>}
  n16 = MaterialFunctionCall(XY=Multiply(A=TextureCoordinate(), B=AppendVector(A=@n15.R, B=@n15.R))) {materialFunction=&Swizzle}
  n17 = NamedRerouteUsage()
  n18 = AppendVector(A=@n15.B, B=@n15.B)
  n19 = MaterialFunctionCall(
   Transition Phase=DynamicParameter() {defaultValue=<rgba 1 -0.5 1 1>, paramNames [Dissolve, Top_Dissolve, Alll speed, Param4]}.Dissolve
   Height Texture=TextureSampleParameter2D(UVs=Panner(Coordinate=@n16.YX, Time=@n17, Speed=@n18) {speedX=1.3}) {parameterName=DissolveTex, samplerType=SAMPLERTYPE_Grayscale, texture=/Game/Global/Textures/Default/Noise/TilingNoise06}.RGB
  ) {materialFunction=&HeightLerp}
  n20 = Multiply(A=Multiply(A=Multiply(A=@n1.A, B=@n6), B=Saturate(None=@n14)), B=Saturate(None=Multiply(A=@n19.Alpha) {constB=3}))
  n21 = MaterialFunctionCall(XY=TextureCoordinate() {uTiling=2, vTiling=1.2}) {materialFunction=&Swizzle}
  n22 = NamedRerouteUsage()
  n23 = ComponentMask(None=TextureCoordinate()) {g=true}
  n24 = OneMinus(None=@n23)
  n25 = Multiply(A=Power(Base=@n23) {constExponent=2.3}, B=Power(Base=Multiply(A=@n24) {constB=4}))
  n26 = Add(
   A=TextureSample(UVs=Panner(Coordinate=@n21.YX, Time=@n22) {speedX=2}) {texture=/BRCosmetics/Weapons/FORT_Melee/Pickaxe_SoapPocket/FX/T_Pickaxe_SoapPocket_Noise}.R
   B=@n25
  )
  n27 = Add(
   A=TextureCoordinate()
   B=AppendVector(A=Constant(), B=DynamicParameter() {defaultValue=<rgba 1 -0.5 1 1>, paramNames [Dissolve, Top_Dissolve, Alll speed, Param4]}.Top_Dissolve)
  )
  n28 = Multiply(A=@n20, B=Add(A=Multiply(A=@n26, B=@n25), B=ComponentMask(None=@n27) {g=true}))
  n29 = ScalarParameter() {parameterName=Use_TopDissolve}
  n30 = ScalarParameter() {defaultValue=1, parameterName=MeshFade}
  DepthFade(Opacity=Multiply(
    A=LinearInterpolate(A=@n20, B=@n28, Alpha=@n29)
    B=MaterialFunctionCall("FadeSensitivity (optional)"=@n30) {materialFunction=&FadeBasedOnViewAngle}.Result
   )) {fadeDistanceDefault=5}
  n31 = LinearInterpolate(
   A=TextureCoordinate() {uTiling=8, vTiling=9}
   B=TextureCoordinate() {uTiling=8, vTiling=5}
   Alpha=ScalarParameter() {parameterName=Use_TopUV}
  )
  n32 = Time()
  n33 = ScalarParameter() {defaultValue=30, parameterName=SnowflakeSpeed}
  n34 = Multiply(A=Floor(None=Multiply(A=@n32, B=@n33))) {constB=0.5}
  n35 = Multiply(
   A=Constant3Vector() {constant=<rgba 0.4265 1.2245 2 1>}
   B=TextureSample(UVs=Panner(Coordinate=@n31, Time=@n34) {speedY=5}) {texture=&/T_Emote_MajorSpeech_Snow}.R
  ) {constB=2}
  n36 = Constant() {r=0.1}
  n37 = Constant()
  n38 = Power(Base=OneMinus(None=Fresnel(ExponentIn=@n36, BaseReflectFractionIn=@n37))) {constExponent=1}
  n39 = Constant3Vector() {constant=<rgba 0.1417 0.314 0.4 1>}
  n40 = LinearInterpolate(
   A=TextureCoordinate() {uTiling=12, vTiling=15}
   B=TextureCoordinate() {uTiling=8, vTiling=5}
   Alpha=ScalarParameter() {parameterName=Use_TopUV}
  )
  n41 = NamedRerouteUsage()
  n42 = Multiply(A=TextureSample(UVs=Panner(Coordinate=@n40, Time=@n41) {speedX=1, speedY=6}) {texture=&T_Emote_MajorSpeech_Snow}.R)
  n43 = LinearInterpolate(
   A=TextureCoordinate() {uTiling=3, vTiling=5}
   B=TextureCoordinate() {uTiling=2, vTiling=0.5}
   Alpha=ScalarParameter() {parameterName=Use_TopUV}
  )
  n44 = NamedRerouteUsage()
  n45 = Multiply(
   A=Add(A=Multiply(A=@n35, B=@n38), B=Multiply(A=@n39, B=@n42))
   B=TextureSample(UVs=Panner(Coordinate=@n43, Time=@n44) {speedX=0.2, speedY=3}) {samplerType=SAMPLERTYPE_Grayscale, texture=/BRCosmetics/Weapons/FORT_Melee/Pickaxe_Cube_Ninja_Male/FX/T_CellNoise}.R
  )
  n46 = ScalarParameter() {parameterName=Use_TopUV}
  n47 = LinearInterpolate(A=TextureCoordinate() {uTiling=1.5, vTiling=1.5}, B=TextureCoordinate() {vTiling=0.5}, Alpha=@n46)
  n48 = NamedRerouteUsage()
  n49 = Power(Base=TextureSample(UVs=Panner(Coordinate=@n47, Time=@n48) {speedX=0.15, speedY=1}) {texture=/Game/Global/Textures/Default/PointSplatter/T_Tiling_Dust}.RGB) {constExponent=1.3}
  n50 = TextureCoordinate() {vTiling=1.2}
  n51 = TextureCoordinate() {uTiling=2, vTiling=0.5}
  n52 = TextureSample(UVs=Panner(Coordinate=LinearInterpolate(A=@n50, B=@n51, Alpha=@n46), Time=@n48) {speedX=0.15, speedY=1}) {texture=/Game/Characters/Player/Female/Medium/Bodies/F_MED_Celestial_Glow/Textures/T_FN_Nebula_LighterBlue}
  n53 = TextureCoordinate() {vTiling=1.2}
  n54 = TextureCoordinate() {vTiling=0.3}
  n55 = ScalarParameter() {parameterName=Use_TopUV}
  n56 = Time()
  n57 = ScalarParameter() {defaultValue=1, parameterName=Speed_Top}
  n58 = TextureSample(UVs=Panner(Coordinate=LinearInterpolate(A=@n53, B=@n54, Alpha=@n55), Time=Multiply(A=@n56, B=@n57)) {speedX=0.1, speedY=1.5}) {samplerType=SAMPLERTYPE_LinearGrayscale, texture=/Game/Athena/Prototype/Terrain/T_CloudCoverage}
  n59 = Multiply(A=Multiply(A=LinearInterpolate(A=@n49, B=@n52.RGB) {constAlpha=0.3}, B=@n58.R)) {constB=3}
  n60 = VectorParameter() {parameterName=PointColor, defaultValue=<rgba 1.1458 3.8313 5 1>}
  n61 = LinearInterpolate(
   A=TextureCoordinate() {uTiling=10, vTiling=11}
   B=TextureCoordinate() {uTiling=18, vTiling=10}
   Alpha=ScalarParameter() {parameterName=Use_TopUV}
  )
  n62 = ScalarParameter() {defaultValue=1, parameterName=UV_Scale}
  n63 = TextureSample(UVs=Panner(Coordinate=Multiply(A=@n61, B=@n62), Time=NamedRerouteUsage()) {speedX=0.2, speedY=6}) {samplerType=SAMPLERTYPE_Masks, texture=&/T_Emote_MajorSpeech_Star01}
  n64 = VectorParameter() {parameterName=BaseTexture, defaultValue=<rgba 2 2.2 0 2>}
  n65 = Multiply(A=TextureCoordinate(), B=AppendVector(A=@n64.R, B=@n64.R))
  n66 = TextureCoordinate() {vTiling=1.2}
  n67 = Multiply(
   A=TextureSample(UVs=Panner(Coordinate=@n66) {speedX=0.3, speedY=1.5}) {samplerType=SAMPLERTYPE_LinearGrayscale, texture=/Game/Effects/Fort_Effects/Emotes/UltraLight/Texture/T_DirectionalNoise_Greyscale}.R
   B=ScalarParameter() {defaultValue=0.15, parameterName=Base_NoiseMul}
  ) {constB=0.15}
  n68 = TextureSampleParameter2D(UVs=Panner(Coordinate=Add(A=@n65, B=@n67), Time=NamedRerouteUsage(), Speed=AppendVector(A=@n64.B, B=@n64.B)) {speedY=2}) {parameterName=BaseTexture, texture=/BRCosmetics/Gadgets/Assets/VinderTech_GliderChute/Glider_SlickSwish/FX/T_Glider_SlickSwish_noise}
  n69 = Multiply(A=@n68.R, B=@n58.R)
  n70 = Multiply(
   A=VectorParameter() {parameterName=BaseColor, defaultValue=<rgba 0.3177 0.7364 1 1>}.RGB
   B=Multiply(A=Power(Base=@n69) {constExponent=1.5}) {constB=0.3}
  )
  n71 = LinearInterpolate(Alpha=@n5.R) {constA=-0.6}
  n72 = Multiply(
   A=LinearInterpolate(
    A=VectorParameter() {parameterName=BaseColorA, defaultValue=<rgba 0.014 0.0573 0.1 1>}.RGB
    B=VectorParameter() {parameterName=BaseColorB, defaultValue=<rgba 0.0531 0.1177 0.15 1>}.RGB
    Alpha=Saturate(None=@n71)
   )
   B=@n1.RGB
  )
  n73 = Add(
   A=Multiply(A=VectorParameter() {parameterName=NoiseSnow_color, defaultValue=<rgba 0.3674 1.2837 2 1>}.RGB, B=Power(Base=@n59) {constExponent=1.5})
   B=Add(A=Multiply(A=@n60.RGB, B=@n63.G), B=Add(A=@n70, B=@n72))
  )
  MaterialFunctionCall(Emissive=LinearInterpolate(A=Add(A=@n45, B=@n73), B=@n73, Alpha=ScalarParameter() {parameterName=Use_NoSnowflake})) {materialFunction=&MF_FX_EmissiveAdjustments}
  n74 = Time()
  n75 = DynamicParameter() {defaultValue=<rgba 1 -0.5 1 1>, paramNames [Dissolve, Top_Dissolve, Alll speed, Param4]}
  n76 = ScalarParameter() {parameterName=Use_DynSpeed}
  NamedRerouteDeclaration(None=Multiply(A=LinearInterpolate(A=@n74, B=@n75.Alll speed, Alpha=@n76), B=ScalarParameter() {defaultValue=1, parameterName=Time})) {name=Speed, nodeColor=<rgba 0 1 0.4141 1>}
  n77 = Multiply(A=Multiply(A=VertexNormalWS(), B=@n5.R)) {constB=6}
  n78 = Constant() {r=1}
  n79 = Constant() {r=1}
  n80 = Panner(Coordinate=Multiply(A=TextureCoordinate() {vTiling=1.3}, B=AppendVector(A=@n78, B=@n79)), Time=NamedRerouteUsage()) {speedX=0.1, speedY=1}
  n81 = Multiply(A=Multiply(
    A=VertexNormalWS()
    B=TextureSample(UVs=@n80) {samplerType=SAMPLERTYPE_Masks, texture=/Game/Accessories/FORT_Backpacks/M_MED_Heist/Textures/T_CrystalNoise_M}.R
   )) {constB=12}
  Reroute(None=Multiply(A=Add(A=@n77, B=@n81), B=ScalarParameter() {defaultValue=1, parameterName=WPOTotal_Intensity}))
 ]
}
