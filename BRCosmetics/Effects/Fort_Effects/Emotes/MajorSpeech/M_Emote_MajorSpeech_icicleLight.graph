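# Additive unlit material for Niagara mesh particles, giving the MajorSpeech emote's icicle light a glowing, scrolling cyan ice look.
# Emissive sums Fresnel-shaped bands and panning noise and dust layers, with one ice noise stepped in time for flicker.
# Opacity combines a noise-distorted ramp, V-edge mask, particle-driven HeightLerp dissolve, view-angle fade and depth fade.
# A shared Speed reroute blends time with a dynamic parameter, and two panning noises build offsets along vertex normals.
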Material /BRCosmetics/Effects/Fort_Effects/Emotes/MajorSpeech/M_Emote_MajorSpeech_icicleLight
graph {
 properties {
  bAutomaticallySetUsageInEditor=false
  bUsedWithNiagaraMeshParticles=true
  blendMode=BLEND_Additive
  shadingModel=MSM_Unlit
  usedShadingModels=Unlit
 }
 nodes [
  n1 = ParticleColor()
  n2 = TextureCoordinate() {uTiling=0}
  n3 = Panner(Coordinate=TextureCoordinate() {uTiling=4}, Time=NamedRerouteUsage(), Speed=Constant2Vector() {g=1}) {speedY=1}
  n4 = Multiply(A=LinearInterpolate(Alpha=TextureSample(UVs=@n3) {samplerType=SAMPLERTYPE_LinearColor, texture=&T_NoisePack_LinearColor}.R) {constA=0.3, constB=0.5}) {constB=0.1}
  n5 = TextureSample(UVs=Add(A=Add(A=@n2, B=@n4), B=Constant2Vector()) {constB=0}) {samplerSource=SSM_Clamp_WorldGroupSettings, texture=/Game/Effects/Fort_Effects/Textures/Ramps/T_Power_Source_Sphere_Ramp}
  n6 = Panner(Coordinate=TextureCoordinate() {uTiling=3}, Time=NamedRerouteUsage(), Speed=Constant2Vector() {g=1.5}) {speedY=1}
  n7 = Multiply(
   A=Power(Base=Multiply(A=@n5.R) {constB=2}) {constExponent=1}
   B=Multiply(A=TextureSample(UVs=@n6) {samplerType=SAMPLERTYPE_LinearGrayscale, texture=&T_SimplePerlin_Greyscale}.RGB)
  )
  n8 = ComponentMask(None=TextureCoordinate()) {g=true}
  n9 = Multiply(A=@n8, B=OneMinus(None=@n8))
  n10 = Saturate(None=Power(Base=Multiply(A=@n9) {constB=5}) {constExponent=6})
  n11 = Multiply(A=@n1.A, B=Saturate(None=Add(A=@n7, B=@n10)))
  n12 = DynamicParameter() {paramNames [Dissolve,Param2,Param3,Param4]}
  n13 = TextureCoordinate() {uTiling=2.2}
  n14 = TextureSampleParameter2D(UVs=Panner(
    Coordinate=MaterialFunctionCall(XY=@n13) {materialFunction=&Swizzle}.YX
    Time=NamedRerouteUsage()
   ) {speedX=1.5}) {parameterName=Dissolvetex, samplerType=SAMPLERTYPE_Grayscale, texture=/Game/Global/Textures/Default/Noise/TilingNoise06}
  n15 = Saturate(None=Multiply(A=MaterialFunctionCall(Transition Phase=@n12.Dissolve, Height Texture=@n14.R) {materialFunction=&HeightLerp}.Alpha) {constB=3})
  n16 = ScalarParameter() {defaultValue=3.5, parameterName=MeshFade}
  DepthFade(
   Opacity=Multiply(
    A=Multiply(A=@n11, B=@n15)
    B=MaterialFunctionCall("FadeSensitivity (optional)"=@n16) {materialFunction=&FadeBasedOnViewAngle}.Result
   )
   FadeDistance=ScalarParameter() {defaultValue=2, parameterName=DepthFade}
  )
  n17 = Constant3Vector() {constant=<rgba 0.28 0.8153 1 1>}
  n18 = Constant() {r=6}
  n19 = Constant()
  n20 = Constant() {r=0.5}
  n21 = Constant()
  n22 = Multiply(A=OneMinus(None=Fresnel(ExponentIn=@n20, BaseReflectFractionIn=@n21))) {constB=4}
  n23 = Multiply(A=Multiply(A=Fresnel(ExponentIn=@n18, BaseReflectFractionIn=@n19), B=Power(Base=@n22) {constExponent=4})) {constB=20}
  n24 = OneMinus(None=Fresnel(ExponentIn=Constant() {r=0.5}, BaseReflectFractionIn=Constant()))
  n25 = Multiply(A=Fresnel(ExponentIn=Constant() {r=3}, BaseReflectFractionIn=Constant()), B=Power(Base=Multiply(A=@n24) {constB=4}) {constExponent=5})
  n26 = TextureCoordinate() {uTiling=3, vTiling=2}
  n27 = Time()
  n28 = Constant() {r=30}
  n29 = Multiply(A=Floor(None=Multiply(A=@n27, B=@n28))) {constB=0.2}
  n30 = Power(Base=TextureSample(UVs=Panner(Coordinate=@n26, Time=@n29) {speedY=3}) {texture=/Game/Global/Textures/Default/Noise/T_UI_Ice_Noise}.RGB) {constExponent=3}
  n31 = TextureSample(UVs=Panner(Coordinate=TextureCoordinate() {uTiling=4, vTiling=2}, Time=NamedRerouteUsage()) {speedX=0.15, speedY=3}) {texture=/Game/Global/Textures/Default/PointSplatter/T_Tiling_Dust}
  n32 = Multiply(
   A=Constant3Vector() {constant=<rgba 0.3229 1.5697 2 1>}
   B=Multiply(A=Multiply(A=@n25) {constB=8}, B=LinearInterpolate(A=@n30, B=@n31.RGB) {constAlpha=0.2})
  )
  n33 = Panner(Coordinate=TextureCoordinate() {uTiling=3, vTiling=2}, Time=NamedRerouteUsage()) {speedY=3}
  n34 = TextureCoordinate() {uTiling=4, vTiling=3}
  n35 = NamedRerouteUsage()
  n36 = LinearInterpolate(
   A=Power(Base=TextureSample(UVs=@n33) {texture=/Game/Global/Textures/Default/Noise/T_Tiling_Noise_01}.R) {constExponent=1.2}
   B=TextureSample(UVs=Panner(Coordinate=@n34, Time=@n35) {speedX=0.15, speedY=1}) {texture=&T_Tiling_Dust}.R
  )
  n37 = Multiply(
   A=LinearInterpolate(A=Constant3Vector() {constant=<rgba 0 0.0224 0.0469 1>}, B=Constant3Vector() {constant=<rgba 0 3.3347 5 1>}, Alpha=Power(Base=@n36))
   B=@n1.RGB
  )
  MaterialFunctionCall(Emissive=Add(A=Multiply(A=@n17, B=@n23), B=Add(A=@n32, B=@n37))) {materialFunction=&MF_FX_EmissiveAdjustments}
  n38 = Time()
  n39 = DynamicParameter() {defaultValue=<rgba 1 -0.5 1 1>, paramNames [Dissolve, Top_Dissolve, Alll speed, Param4]}
  n40 = ScalarParameter() {parameterName=Use_DynSpeed}
  NamedRerouteDeclaration(None=Multiply(A=LinearInterpolate(A=@n38, B=@n39.Alll speed, Alpha=@n40), B=ScalarParameter() {defaultValue=1, parameterName=Time})) {name=Speed, nodeColor=<rgba 0 1 0.4141 1>}
  n41 = Multiply(A=TextureCoordinate() {vTiling=1.3}, B=AppendVector(A=Constant() {r=1}, B=Constant() {r=0.5}))
  n42 = NamedRerouteUsage()
  n43 = Multiply(
   A=VertexNormalWS()
   B=TextureSample(UVs=Panner(Coordinate=@n41, Time=@n42) {speedX=0.1, speedY=1}) {samplerType=SAMPLERTYPE_LinearGrayscale, texture=/Game/Packages/Fortress_Sky/SkyDome/Master/T_SkyDome_Cloud_Voroni}.R
  )
  n44 = Multiply(A=TextureCoordinate() {vTiling=1.3}, B=AppendVector(A=Constant() {r=1}, B=Constant() {r=1}))
  n45 = NamedRerouteUsage()
  n46 = Multiply(
   A=VertexNormalWS()
   B=TextureSample(UVs=Panner(Coordinate=@n44, Time=@n45) {speedX=0.1, speedY=1}) {samplerType=SAMPLERTYPE_Masks, texture=/Game/Accessories/FORT_Backpacks/M_MED_Heist/Textures/T_CrystalNoise_M}.R
  )
  Reroute(None=Add(A=Multiply(A=@n43) {constB=6}, B=Multiply(A=@n46) {constB=12}))
 ]
}
