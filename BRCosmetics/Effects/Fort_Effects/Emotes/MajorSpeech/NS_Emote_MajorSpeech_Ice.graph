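# Niagara effect for the ice variant of the MajorSpeech emote, surrounding its owner with swirling icy air, smoke and glints.
# Mesh emitters build rotating, expanding airflow, dark and smoke rings, while sprite flares twinkle on a cylinder shell.
# GPU Ice_Shards randomly tags particles as snowflakes or shards through renderer visibility, drifting them with curl noise, gravity and drag.
# Airflow and smoke tint by time of day, and smoke scale adjusts for front end.
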NiagaraSystem /BRCosmetics/Effects/Fort_Effects/Emotes/MajorSpeech/NS_Emote_MajorSpeech_Ice
dependencies {
 usedDataInterfaces [
  &NiagaraDataInterfaceCurve
  &NiagaraDataInterfaceMeshRendererInfo
  &NiagaraDataInterfaceCamera
  &NiagaraDataInterfaceVectorCurve
 ]
 usedDynamicInputs []
 usedModules [&SystemState]
}
emitters {
 AirflowRemain {
  emitterUpdateScript [
   &EmitterState(Life Cycle Mode=::NewEnumerator1, Loop Behavior=::NewEnumerator1, Loop Delay=0.5, Loop Duration=0.8, UseLoopDelay=true)
   &SpawnRate(
    SpawnRate=&Multiply_Float(
     A=35
     B=&FloatFromCurve(FloatCurve=DataInterface() {propertyValues {Curve {keys [<at 0 0>,<at 0.4 1>]}}}, CurveIndex=$Emitter.NormalizedLoopAge)
    )
    Use Spawn Probability=false
   )
   &SpawnBurst_Instantaneous(Spawn Count=2, Use Spawn Probability=false, UseLoopCountLimit=false)
  ]
  particleSpawnScript [
   &InitializeParticle(
    Color=<rgba 0.8542 0.9725 1 1>
    Lifetime=0.5
    Mesh Scale=<1 1 2>
    Mesh Scale Mode=::NewEnumerator4
    Position Offset=<0 0 30>
    UsePositionOffset=true
    Write Material Random=true
   )
   &ShapeLocation(
    Handle Radius=0
    Large Radius=20
    MaskSpawnGroup=false
    Rotation Mode=::NewEnumerator1
    Shape Primitive=::NewEnumerator3
    Torus Distribution Mode:ENiagaraTorusDistributionMode=::NewEnumerator1
    U Position=&ReturnNormalizedExecIndex()
    UseOwnerScale=true
   )
   &InitialMeshOrientation(
    Mesh Orientation Mode=::NewEnumerator2
    Rotation=&RandomRangeVector(Minimum=<0 0 0>, Maximum=<0 0 1>)
    Use Rotation Vector=true
   )
   &InitialMeshOrientation(Mesh Orientation Mode=::NewEnumerator2, Use Rotation Vector=true) {module=InitialMeshOrientation001}
  ]
  particleUpdateScript [
   &ParticleState()
   &UpdateMeshOrientation(Rotation Rate=0.8, Rotation Vector=<0 0 1>)
   &ScaleMeshSize(Scale Factor=&Multiply_VectorByFloat(
     Vector=&VectorFromCurve(
      VectorCurve=DataInterface() {propertyValues {XCurve {keys [<at 0 0.6>,<at 1 1.4>]}, YCurve {keys [<at 0 0.6>,<at 1 1.4>]}, ZCurve {keys [<at 0 1>,<at 1 1>]}}}
      Scale Curve=&Multiply_VectorByFloat(Vector=<1 1 1>, Float=1)
     )
     Float=1
    ))
   &DynamicMaterialParameters(
    Index 0 Param 1=&FloatFromCurve(FloatCurve=DataInterface() {propertyValues {Curve {keys [<at 0 0.1>,<at 1 0.5>]}}})
    Index 0 Param 2=&RandomRangeFloat(Minimum=40, Maximum=60)
    Module.Index0Param0Write=true
    Module.Index0Param1Write=true
    Module.Index0Param2Write=true
    Module.Index0Param3Write=true
   )
   &ScaleColor(
    Scale RGB=&Multiply_VectorByFloat(
     Vector=&MakeVectorFromLinearColorRGB(Color=&FNTimeOfDayColor(
       Night Color=<rgba 0.474 0.6858 1 1>
       Morning Color=<rgba 0.75 0.8507 1 1>
       Day Color=<rgba 0.75 0.8507 1 1>
       Evening Color=<rgba 0.75 0.8507 1 1>
      ))
     Float=&FloatFromCurve(
      FloatCurve=DataInterface() {propertyValues {Curve {keys [
          <at 0 0 RCIM_Cubic>
          <at 0.2 1 RCIM_Cubic>
          {arriveTangent=-0.6427, interpMode=RCIM_Cubic, leaveTangent=-0.6427, tangentMode=RCTM_User, time=1, value=0}
         ]}}}
      Scale Curve=&Multiply_Float(A=1, B=&FNTimeOfDayScalar(Night Scale=0.8, Morning Scale=0.8, Evening Scale=0.8))
     )
    )
    ScaleA=true
    ScaleRGB=true
   )
  ]
  rendererClasses [&NiagaraMeshRendererProperties]
  renderers [{rendererClass=&NiagaraMeshRendererProperties, rendererIndex=0}]
 }
 Disappear_RingDark {
  emitterUpdateScript [
   &EmitterState(Life Cycle Mode=::NewEnumerator1, Loop Behavior=::NewEnumerator1, Loop Duration=1.5, UseLoopDelay=true)
   &SpawnBurst_Instantaneous(Spawn Count=2, Use Spawn Probability=false, UseLoopCountLimit=false)
   &SpawnRate(
    SpawnRate=&Multiply_Float(
     A=10
     B=&FloatFromCurve(FloatCurve=DataInterface() {propertyValues {Curve {keys [<at 0 1>,<at 1 1.5>]}}}, CurveIndex=$Emitter.NormalizedLoopAge)
    )
    Use Spawn Probability=false
   )
  ]
  particleSpawnScript [
   &InitializeParticle(
    Lifetime Max=1
    Lifetime Mode=::NewEnumerator1
    Mesh Scale Mode=::NewEnumerator1
    Mesh Uniform Scale Max=1.2
    UsePositionOffset=true
    Write Material Random=true
   )
   &InitialMeshOrientation(Mesh Orientation Mode=::NewEnumerator2, Rotation=<0.25 0 0>, Use Rotation Vector=true)
   &InitialMeshOrientation(
    Mesh Orientation Mode=::NewEnumerator2
    Rotation=&RandomRangeVector(Minimum=<0 0 0>, Maximum=<0 0 1>)
    Use Rotation Vector=true
   ) {module=InitialMeshOrientation001}
   &InitialMeshOrientation(Mesh Orientation Mode=::NewEnumerator2, Rotation=<0 0.03 0>, Use Rotation Vector=true) {module=InitialMeshOrientation002}
   &CameraOffset()
   &ShapeLocation(
    Cylinder Radius=1
    MaskSpawnGroup=false
    Offset=<0 0 20>
    Offset Mode=::NewEnumerator0
    Override Seed=false
    Shape Primitive=::NewEnumerator1
    Surface Only=false
    UseOwnerScale=true
   )
   &AddVelocity(Velocity=<10 0 50>)
  ]
  particleUpdateScript [
   &ParticleState()
   &UpdateMeshOrientation(
    Rotation Coordinate Space=::NewEnumerator2
    Rotation Rate=&Multiply_Float(
     A=0.5
     B=&FloatFromCurve(FloatCurve=DataInterface() {propertyValues {Curve {keys [<at 0.2 1 RCIM_Cubic>,<at 0.5 3 RCIM_Cubic>]}}})
    )
    Rotation Vector=<0 0 1>
   )
   &ScaleMeshSize(Scale Factor=&Multiply_VectorByFloat(
     Vector=<1 1 1>
     Float=&Multiply_Float(
      A=0.8
      B=&FloatFromCurve(FloatCurve=DataInterface() {propertyValues {Curve {keys [{arriveTangent=0.4615, time=0, value=0.6}, {arriveTangent=0.2118, leaveTangent=0.2118, time=1, value=1}]}}})
     )
    ))
   &ScaleColor(
    Scale Alpha=&FloatFromCurve(
     FloatCurve=DataInterface() {propertyValues {Curve {keys [
         {interpMode=RCIM_Cubic, tangentMode=RCTM_User, time=0, value=0}
         {interpMode=RCIM_Cubic, tangentMode=RCTM_User, time=0.4, value=1}
         {tangentMode=RCTM_User, time=1, value=0}
        ]}}}
     Scale Curve=0.3
    )
    Scale RGB=&MakeVectorFromLinearColorRGB(Color=<rgba 0.5478 0.8641 1 1>)
    ScaleA=true
    ScaleRGB=true
   )
   &SolveForcesAndVelocity(Clamp Velocity=false, Limit Acceleration=false)
  ]
  rendererClasses [&NiagaraMeshRendererProperties]
  renderers [{rendererClass=&NiagaraMeshRendererProperties, rendererIndex=0}]
 }
 Flare {
  emitterUpdateScript [
   &EmitterState(Life Cycle Mode=::NewEnumerator1, Loop Behavior=::NewEnumerator1, Loop Duration=0.6, UseLoopDelay=false)
   &SpawnBurst_Instantaneous(Spawn Count=5, Use Spawn Probability=false, UseLoopCountLimit=false)
   &SpawnRate(SpawnRate=5, Use Spawn Probability=false)
  ]
  particleSpawnScript [
   &InitializeParticle(
    Color=<rgba 0.3125 0.7482 1 1>
    Lifetime Min=1.5
    Lifetime Mode=::NewEnumerator1
    Sprite Size Mode=::NewEnumerator1
    Uniform Sprite Size Max=35
    Uniform Sprite Size Min=25
    UsePositionOffset=false
    Write Material Random=true
   )
   &ShapeLocation(
    Cylinder Height=150
    Cylinder Radius=80
    MaskSpawnGroup=false
    Offset=<0 0 160>
    Offset Mode=::NewEnumerator0
    Override Seed=false
    Shape Primitive=::NewEnumerator1
    Surface Only=true
    Surface Only Band Thickness=0.2
    Use Endcaps In Surface Only Mode=false
    UseOwnerScale=true
   )
  ]
  particleUpdateScript [
   &ParticleState()
   &ScaleSpriteSize(Uniform Curve Sprite Scale=DataInterface() {propertyValues {Curve {keys [
        <at 0 0 RCIM_Cubic>
        {arriveTangent=0.2, interpMode=RCIM_Cubic, leaveTangent=0.2, time=0.2, value=0.7}
        {arriveTangent=0.4985, interpMode=RCIM_Cubic, leaveTangent=0.4985, time=0.5, value=0.1}
        {arriveTangent=-0.2, interpMode=RCIM_Cubic, leaveTangent=-0.2, time=0.763, value=0.9807}
        <at 1 0 RCIM_Cubic>
       ]}}})
   &ScaleColor(
    Scale Alpha=&FloatFromCurve(FloatCurve=DataInterface() {propertyValues {Curve {keys [<at 0 0>,<at 0.2 1>,<at 0.8 1>,<at 1 0>]}}})
    ScaleA=true
    ScaleRGB=true
   )
   /SetVariables(Particles.ColorGlow:LinearColor=&MakeLinearColorFromVectorAndFloat(
     "Vector (RGB)"=&MakeVectorFromLinearColorRGB(Color=<rgba 0.1785 0.5679 1 1>)
     "Float (Alpha)"=&FloatFromCurve(
      FloatCurve=DataInterface() {propertyValues {Curve {keys [
          <at 0 0>
          {arriveTangent=1.25, leaveTangent=1.25, time=0.2028, value=0.9955}
          {arriveTangent=-1.2487, leaveTangent=-1.2487, time=0.8, value=1}
          <at 1 0>
         ]}}}
      Scale Curve=0.11
     )
    )) {isSetParameters=true}
   /SetVariables(Particles.SizeGlow:Vector2f=&Multiply_Vector2D_ByFloat(Float=0.5, Vector2D=$Particles.SpriteSize)) {isSetParameters=true}
  ]
  rendererClasses [&NiagaraSpriteRendererProperties]
  renderers [
   {rendererClass=&NiagaraSpriteRendererProperties}
   {rendererClass=&NiagaraSpriteRendererProperties, rendererIndex=1}
  ]
 }
 Ice_Shards {
  simTarget=GPUComputeSim
  emitterUpdateScript [
   &EmitterState(Life Cycle Mode=::NewEnumerator1, Loop Behavior=::NewEnumerator1, Loop Duration=0.8, UseLoopDelay=false)
   &SpawnBurst_Instantaneous(Spawn Count=20, Use Spawn Probability=false, UseLoopCountLimit=false)
   &SpawnRate(
    Spawn Probability=&FloatFromCurve(
     FloatCurve=DataInterface() {propertyValues {Curve {keys [
         {interpMode=RCIM_Cubic, tangentMode=RCTM_User, time=0.1, value=0}
         {interpMode=RCIM_Cubic, tangentMode=RCTM_User, time=1, value=1}
         {arriveTangent=0.0061, interpMode=RCIM_Cubic, leaveTangent=-0.1111, tangentMode=RCTM_Break, time=9, value=1}
         <at 10.6 0 RCIM_Cubic>
        ]}}}
     CurveIndex=$Emitter.Age
    )
    SpawnRate=100
    Use Spawn Probability=false
   )
  ]
  particleSpawnScript [
   /SetVariables(
    Particles.Particle_Depth:NiagaraFloat=&FloatFromCurve(
     FloatCurve=DataInterface() {propertyValues {Curve {keys [
         {arriveTangent=1.2753, interpMode=RCIM_Cubic, leaveTangent=1.2753, tangentMode=RCTM_User, time=0, value=0}
         {arriveTangent=0.204, interpMode=RCIM_Cubic, leaveTangent=0.204, tangentMode=RCTM_User, time=0.9, value=1}
        ]}}}
     CurveIndex=&RandomRangeFloat(Minimum=0, Maximum=1)
    )
    Particles.Particle_Height:NiagaraFloat=&FloatFromCurve(
     FloatCurve=DataInterface() {propertyValues {Curve {keys [
         {arriveTangent=1.9477, interpMode=RCIM_Cubic, leaveTangent=1.9477, tangentMode=RCTM_User, time=0, value=0}
         {arriveTangent=0.1423, interpMode=RCIM_Cubic, leaveTangent=0.1423, tangentMode=RCTM_User, time=0.5, value=0.5}
         {arriveTangent=1.4427, interpMode=RCIM_Cubic, leaveTangent=1.4427, tangentMode=RCTM_User, time=1, value=1}
        ]}}}
     CurveIndex=&RandomRangeFloat(Minimum=0, Maximum=1)
    )
    Particles.Particle_Scale:NiagaraFloat=&RandomRangeFloat(Minimum=0, Maximum=1)
   ) {isSetParameters=true}
   &InitializeParticle(
    Color Maximum=<rgba 0.9 0.9833 1 1>
    Color Minimum=<rgba 0.6 0.8667 1 1>
    Color Mode=::NewEnumerator2
    Mass Mode=::NewEnumerator3
    Mesh Renderer Array Visibility Mode=::NewEnumerator3
    Mesh Renderer Info=DataInterface() {propertyValues {MeshRenderer=&self:Ice_Shards_0.NiagaraMeshRendererProperties_0}}
    Mesh Scale Mode=::NewEnumerator1
    Mesh Uniform Scale Max=0.23
    Mesh Uniform Scale Min=0.15
    Sprite Rotation Mode=::NewEnumerator2
    Sprite Size Mode=::NewEnumerator1
    Uniform Sprite Size Max=8
    Uniform Sprite Size Min=2
    UsePositionOffset=false
    Write Material Random=true
   )
   &ShapeLocation(
    Cylinder Height=260
    MaskSpawnGroup=false
    Offset=<0 0 120>
    Offset Mode=::NewEnumerator0
    Override Seed=false
    Shape Primitive=::NewEnumerator1
    Surface Only=false
    UseOwnerScale=true
   )
   &CameraOffset(Camera Offset Amount=6)
   &InitialMeshOrientation(
    Mesh Orientation Mode=::NewEnumerator2
    Rotation=&RandomRangeVector(Minimum=<0 0 0>, Maximum=<1 1 1>)
    Use Rotation Vector=true
   )
   /SetVariables(Particles.RotationDirection:Vector3f=&Lerp_Vector(
     A=&RandomRangeVector(Minimum=<-0.6 -0.6 1>, Maximum=<0.8 0.8 1>)
     B=&RandomRangeVector(Minimum=<-0.2 -0.2 1>, Maximum=<0.3 0.3 1>)
     Alpha=&FloatFromCurve(
      FloatCurve=DataInterface() {propertyValues {Curve {keys [
          {interpMode=RCIM_Cubic, tangentMode=RCTM_User, time=0, value=0}
          {arriveTangent=0.5342, interpMode=RCIM_Cubic, leaveTangent=0.5342, tangentMode=RCTM_User, time=1, value=1}
         ]}}}
      CurveIndex=$Particles.Particle_Depth
     )
    )) {isSetParameters=true}
   /SetVariables(Particles.isSnowFlake:NiagaraBool=&RandomBool()) {isSetParameters=true}
   /SetVariables(Particles.VisibilityTag=&MakeCustomIntFromBool(Bool=$Particles.isSnowFlake, True INT=0, False INT=1)) {isSetParameters=true}
   &DynamicMaterialParameters(
    Index 0 Param 1=&RandomRangeFloat(Minimum=1, Maximum=2)
    Module.Index0Param0Write=true
    Module.Index0Param1Write=false
    Module.Index0Param2Write=false
    Module.Index0Param3Write=false
   )
   &AddVelocity(Module.Use Falloff Distance=false, Override Seed=false, Velocity Mode=::NewEnumerator1, Velocity Speed=0)
  ]
  particleUpdateScript [
   &ParticleState()
   &SpriteRotationRate(Rotation Rate=&RandomRangeFloat(Minimum=50, Maximum=150))
   &UpdateMeshOrientation(
    Rotation Rate=&Multiply_Float(
     A=&Lerp_Float(A=1.5, B=1, Alpha=$Particles.Particle_Depth)
     B=&Lerp_Float(A=0.5, B=0.3, Alpha=$Particles.Particle_Scale)
    )
    Rotation Vector=&RandomRangeVector(Minimum=<-1 -1 -1>, Maximum=<1 1 1>)
   )
   &ScaleMeshSize(Scale Factor=&VectorFromFloat(Value=&FloatFromCurve(FloatCurve=DataInterface() {propertyValues {Curve {keys [
          {interpMode=RCIM_Cubic, tangentMode=RCTM_User, time=0, value=0}
          {interpMode=RCIM_Cubic, tangentMode=RCTM_User, time=0.2, value=1}
          {arriveTangent=-3.406, interpMode=RCIM_Cubic, leaveTangent=-3.406, tangentMode=RCTM_User, time=1, value=0}
         ]}}})))
   &ScaleSpriteSize(Uniform Curve Sprite Scale=DataInterface() {propertyValues {Curve {keys [{interpMode=RCIM_Cubic, tangentMode=RCTM_User, time=0, value=0}, {interpMode=RCIM_Cubic, tangentMode=RCTM_User, time=0.1, value=1}]}}})
   &SpriteFacingAndAlignment(
    Sprite Facing=&Lerp_Vector(
     A=&RandomRangeVector(Minimum=<-1 -1 -1>, Maximum=<1 1 1>)
     B=&ReturnCameraVector(Camera Vectors=::NewEnumerator1, Coordinate Space=::Simulation)
     Alpha=&FloatFromCurve(
      FloatCurve=DataInterface() {propertyValues {Curve {keys [
          {arriveTangent=2.0555, interpMode=RCIM_Cubic, leaveTangent=2.0555, tangentMode=RCTM_User, time=0, value=0}
          {interpMode=RCIM_Cubic, tangentMode=RCTM_User, time=1, value=1}
         ]}}}
      CurveIndex=&RandomRangeFloat(Minimum=0, Maximum=1)
     )
    )
    WriteAlignment=false
    WriteFacing=true
   )
   &ScaleColor(
    Scale Alpha=&FloatFromCurve(
     FloatCurve=DataInterface() {propertyValues {Curve {keys [
         <at 0 0>
         {arriveTangent=1.1111, leaveTangent=1.1111, time=0.2, value=1}
         {arriveTangent=-1.1111, leaveTangent=-1.1111, time=0.9, value=1}
         <at 1 0>
        ]}}}
     Scale Curve=&RandomRangeFloat(Minimum=0.1, Maximum=0.5)
    )
    Scale RGB=&MakeVectorFromLinearColorRGB(Color=<rgba 0.5625 0.896 1 1>)
    ScaleA=true
    ScaleRGB=true
   )
   &CurlNoiseForce(Bias Noise=false, Noise Strength=8, Pan Noise=false)
   &VortexForce(UseOriginOffset=false) {enabled=false}
   &GravityForce(Gravity=&RandomRangeVector(Minimum=<0 0 -5>, Maximum=<0 0 -10>))
   &Drag(Drag=0.5, Use Linear Drag=true, Use Rotational Drag=true)
   &InheritSourceMovement(Applied Position Scale=<0.5 0.5 0.5>, ApplyForce=false, ApplyPosition=true, ApplyPositionFalloff=false)
   &SolveForcesAndVelocity(Clamp Velocity=false, Limit Acceleration=false)
   /SetVariables(Particles.Color02:LinearColor=&MakeLinearColorFromVectorAndFloat("Vector (RGB)"=&MakeVectorFromLinearColorRGB(Color=<rgba 0.375 1.1543 1.5 1>), "Float (Alpha)"=1)) {isSetParameters=true}
  ]
  rendererClasses [&NiagaraSpriteRendererProperties, &NiagaraMeshRendererProperties]
  renderers [
   {rendererClass=&NiagaraSpriteRendererProperties}
   {rendererClass=&NiagaraMeshRendererProperties, rendererIndex=1}
  ]
 }
 SmokeRing {
  emitterUpdateScript [
   &EmitterState(Life Cycle Mode=::NewEnumerator1, Loop Behavior=::NewEnumerator1, Loop Duration=1, UseLoopDelay=false)
   &SpawnBurst_Instantaneous(Use Spawn Probability=false, UseLoopCountLimit=false)
   &SpawnRate(SpawnRate=3, Use Spawn Probability=false)
  ]
  particleSpawnScript [
   &InitializeParticle(
    Color=<rgba 0.7344 0.9366 1 1>
    Lifetime Max=1.5
    Lifetime Min=2
    Lifetime Mode=::NewEnumerator1
    Mesh Scale Max=<1 1 3>
    Mesh Scale Min=<1 1 2>
    Mesh Scale Mode=::NewEnumerator2
    UsePositionOffset=true
    Write Material Random=true
   )
   &InitialMeshOrientation(
    Mesh Orientation Mode=::NewEnumerator2
    Rotation=&RandomRangeVector(Minimum=<-0.03 -0.03 0>, Maximum=<0.03 0.03 1>)
    Use Rotation Vector=true
   )
   &ShapeLocation(MaskSpawnGroup=false, Offset=<0 0 10>, Offset Mode=::NewEnumerator0, Override Seed=false, Sphere Radius=1, UseOwnerScale=true)
   &AddVelocity(Velocity=<0 0 30>)
  ]
  particleUpdateScript [
   &ParticleState()
   &DynamicMaterialParameters(
    Index 0 Param 1=&FloatFromCurve(FloatCurve=DataInterface() {propertyValues {Curve {keys [<at 0 1>,<at 1 0>]}}})
    Module.Param0WriteEnabled=true
    Module.Param1WriteEnabled=true
    Module.Param2WriteEnabled=true
    Module.Param3WriteEnabled=true
   )
   &ScaleMeshSize(Scale Factor=&Multiply_VectorByFloat(
     Vector=&Multiply_VectorByFloat(Vector=$Engine.Owner.Scale, Float=&FN_FrontEndFloatModifier(Gameplay=0.9, Front End=0.8))
     Float=&FloatFromCurve(
      FloatCurve=DataInterface() {propertyValues {Curve {keys [
          {arriveTangent=2.0614, interpMode=RCIM_Cubic, leaveTangent=2.0614, tangentMode=RCTM_User, time=0, value=0.5}
          {arriveTangent=0.2273, interpMode=RCIM_Cubic, leaveTangent=0.2273, tangentMode=RCTM_User, time=1, value=1.5}
         ]}}}
      Scale Curve=1.5
     )
    ))
   &UpdateMeshOrientation(
    Rotation Rate=&Multiply_Float(
     A=0.5
     B=&FloatFromCurve(FloatCurve=DataInterface() {propertyValues {Curve {keys [<at 0.2 1 RCIM_Cubic>,<at 0.5 3 RCIM_Cubic>]}}})
    )
    Rotation Vector=<0 0 1>
   )
   &ScaleColor(
    Scale Alpha=&FloatFromCurve(FloatCurve=DataInterface() {propertyValues {Curve {keys [
         {arriveTangent=0.5385, interpMode=RCIM_Cubic, leaveTangent=0.5385, tangentMode=RCTM_User, time=0, value=0}
         {arriveTangent=-1.4305, interpMode=RCIM_Cubic, leaveTangent=-1.4305, tangentMode=RCTM_User, time=0.2, value=1}
         {arriveTangent=-0.1559, tangentMode=RCTM_User, time=1, value=0}
        ]}}})
    Scale RGB=&MakeVectorFromLinearColorRGB(Color=&FNTimeOfDayColor(Night Color=<rgba 0.4356 0.7202 1 1>, Morning Color=<rgba 1 1 1 1>, Day Color=<rgba 1 1 1 1>, Evening Color=<rgba 1 1 1 1>))
    ScaleA=true
    ScaleRGB=true
   )
   &SolveForcesAndVelocity(Clamp Velocity=false, Limit Acceleration=false)
  ]
  rendererClasses [&NiagaraMeshRendererProperties]
  renderers [{rendererClass=&NiagaraMeshRendererProperties, rendererIndex=0}]
 }
}
summary {
 systemName=NS_Emote_MajorSpeech_Ice
 emitters [
  {
   emitterName=Ice_Shards
   simTarget=GPUComputeSim
   rendererClasses [&NiagaraSpriteRendererProperties, &NiagaraMeshRendererProperties]
  }
  {emitterName=SmokeRing, rendererClasses [&NiagaraMeshRendererProperties]}
  {emitterName=Flare, rendererClasses [&NiagaraSpriteRendererProperties]}
  {emitterName=Disappear_RingDark, rendererClasses [&NiagaraMeshRendererProperties]}
  {emitterName=AirflowRemain, rendererClasses [&NiagaraMeshRendererProperties]}
 ]
 userVariables [
  User.bIsFrontEnd {type=NiagaraBool}
  User.bIsFrontEndPreview {type=NiagaraBool}
 ]
}
