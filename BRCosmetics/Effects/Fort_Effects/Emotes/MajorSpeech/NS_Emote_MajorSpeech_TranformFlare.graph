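# Niagara effect for the MajorSpeech emote's transform moment: a brief shower of pale-blue flickering flare sprites on GPU emitters.
# Flare pins particles around five skeletal-mesh bone or socket targets by partition index, while Flare001 fills a downward cone from above and inherits partial source movement.
# Both pair an initial burst with a tapering spawn rate, double-pulse sprite size, and compute separate glow color and size alongside two sprite renderers.
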NiagaraSystem /BRCosmetics/Effects/Fort_Effects/Emotes/MajorSpeech/NS_Emote_MajorSpeech_TranformFlare
dependencies {
 usedDataInterfaces [&NiagaraDataInterfaceCurve, &NiagaraDataInterfaceSkeletalMesh]
 usedDynamicInputs []
 usedModules [
  &self:SystemSpawnScript.SystemScriptSource.SystemScriptGraph.NiagaraNodeAssignment_0.SetVariables_#1
  &SystemState
 ]
}
emitters {
 Flare {
  simTarget=GPUComputeSim
  emitterUpdateScript [
   &EmitterState(Life Cycle Mode=::NewEnumerator1, Loop Behavior=::NewEnumerator1, Loop Duration=1.5, UseLoopDelay=false)
   &SpawnBurst_Instantaneous(Spawn Count=5, Use Spawn Probability=false, UseLoopCountLimit=false)
   &SpawnRate(
    SpawnRate=&Multiply_Float(
     A=30
     B=&FloatFromCurve(FloatCurve=DataInterface() {propertyValues {Curve {keys [<at 0 1>,<at 0.8 1>,<at 1 0>]}}}, CurveIndex=$Emitter.Age)
    )
    Use Spawn Probability=false
   )
  ]
  particleSpawnScript [
   &PartitionParticles(Number Of Partitions=5)
   &InitializeParticle(
    Color=<rgba 0.4844 0.8112 1 1>
    Lifetime Max=0.8
    Lifetime Min=0.4
    Lifetime Mode=::NewEnumerator1
    Sprite Size Mode=::NewEnumerator1
    Uniform Sprite Size Max=22
    Uniform Sprite Size Min=18
    UsePositionOffset=false
    Write Material Random=true
   )
   &ShapeLocation(MaskSpawnGroup=false, Offset Mode=::NewEnumerator0, Override Seed=false, Sphere Radius=25, UseOwnerScale=true)
   &CameraOffset(Camera Offset Amount=20)
  ]
  particleUpdateScript [
   &ParticleState()
   &SkeletalMeshLocation(
    Bone / Socket Index=$Particles.PartitionParticles.Partition
    Bone Sampling Mode=::NewEnumerator4
    MaskSpawnGroup=false
    Position Sampling=::NewEnumerator1
    Sampled Position Offset=&Add_Vector(A=$Particles.ShapeLocation.ShapeVector, B=<0 0 0>)
    Skeletal Mesh=$System.Skeletal Mesh Reader
    Source Movement Scale:Vector3f=<0.6 0.6 0.6>
   )
   &ScaleSpriteSize(
    Uniform Curve Scale=&RandomRangeFloat(Minimum=1.5, Maximum=2.5)
    Uniform Curve Sprite Scale=DataInterface() {propertyValues {Curve {keys [
        <at 0 0 RCIM_Cubic>
        {arriveTangent=0.6667, interpMode=RCIM_Cubic, leaveTangent=0.6667, time=0.2, value=1}
        {arriveTangent=-0.25, interpMode=RCIM_Cubic, leaveTangent=-0.25, time=0.3, value=0.2}
        {arriveTangent=-0.4, interpMode=RCIM_Cubic, leaveTangent=-0.4, time=0.6, value=0.9}
        <at 0.8 0 RCIM_Cubic>
       ]}}}
   )
   &ScaleColor(
    Scale Alpha=&FloatFromCurve(FloatCurve=DataInterface() {propertyValues {Curve {keys [<at 0 0>,<at 0.2 1>,<at 0.8 1>,<at 1 0>]}}})
    ScaleA=true
    ScaleRGB=true
   )
   /SetVariables(Particles.ColorGlow:LinearColor=&MakeLinearColorFromVectorAndFloat(
     "Vector (RGB)"=&MakeVectorFromLinearColorRGB(Color=<rgba 0.1724 0.6122 1 1>)
     "Float (Alpha)"=&FloatFromCurve(
      FloatCurve=DataInterface() {propertyValues {Curve {keys [
          <at 0 0>
          {arriveTangent=1.25, leaveTangent=1.25, time=0.2028, value=0.9955}
          {arriveTangent=-1.2487, leaveTangent=-1.2487, time=0.8, value=1}
          <at 1 0>
         ]}}}
      Scale Curve=0.25
     )
    )) {isSetParameters=true, module=SetVariables_#2}
   /SetVariables(Particles.SizeGlow:Vector2f=&Multiply_Vector2D_ByFloat(Float=0.5, Vector2D=$Particles.SpriteSize)) {isSetParameters=true, module=SetVariables_#3}
  ]
  rendererClasses [&NiagaraSpriteRendererProperties]
  renderers [
   {rendererClass=&NiagaraSpriteRendererProperties}
   {rendererClass=&NiagaraSpriteRendererProperties, rendererIndex=1}
  ]
 }
 Flare001 {
  simTarget=GPUComputeSim
  emitterUpdateScript [
   &EmitterState(Life Cycle Mode=::NewEnumerator1, Loop Behavior=::NewEnumerator1, Loop Duration=1.4, UseLoopDelay=false)
   &SpawnBurst_Instantaneous(Spawn Count=5, Use Spawn Probability=false, UseLoopCountLimit=false)
   &SpawnRate(
    SpawnRate=&Multiply_Float(
     A=60
     B=&FloatFromCurve(FloatCurve=DataInterface() {propertyValues {Curve {keys [<at 0 1>,<at 0.7 1>,<at 1 0>]}}}, CurveIndex=$Emitter.Age)
    )
    Use Spawn Probability=false
   )
  ]
  particleSpawnScript [
   &InitializeParticle(
    Color=<rgba 0.526 0.8264 1 1>
    Lifetime Max=0.8
    Lifetime Min=0.4
    Lifetime Mode=::NewEnumerator1
    Sprite Size Mode=::NewEnumerator1
    Uniform Sprite Size Max=22
    Uniform Sprite Size Min=18
    UsePositionOffset=false
    Write Material Random=true
   )
   &ShapeLocation(
    Cone Angle=40
    Cone Length=150
    MaskSpawnGroup=false
    Offset=<0 0 160>
    Offset Mode=::NewEnumerator0
    Override Seed=false
    Rotation Axis=<0 -0.12 -1>
    Rotation Mode=::NewEnumerator1
    Shape Primitive=::NewEnumerator5
    UseOwnerScale=true
   )
   &CameraOffset(Camera Offset Amount=20)
  ]
  particleUpdateScript [
   &ParticleState()
   &ScaleSpriteSize(
    Uniform Curve Scale=&RandomRangeFloat(Minimum=1.5, Maximum=3)
    Uniform Curve Sprite Scale=DataInterface() {propertyValues {Curve {keys [
        <at 0 0 RCIM_Cubic>
        {arriveTangent=0.4, interpMode=RCIM_Cubic, leaveTangent=0.4, time=0.2, value=1}
        {arriveTangent=-0.1667, interpMode=RCIM_Cubic, leaveTangent=-0.1667, time=0.5, value=0.2}
        {arriveTangent=-0.4, interpMode=RCIM_Cubic, leaveTangent=-0.4, time=0.8, value=0.9}
        <at 1 0 RCIM_Cubic>
       ]}}}
   )
   &ScaleColor(
    Scale Alpha=&FloatFromCurve(FloatCurve=DataInterface() {propertyValues {Curve {keys [<at 0 0>,<at 0.2 1>,<at 0.8 1>,<at 1 0>]}}})
    ScaleA=true
    ScaleRGB=true
   )
   /SetVariables(Particles.ColorGlow:LinearColor=&MakeLinearColorFromVectorAndFloat(
     "Vector (RGB)"=&MakeVectorFromLinearColorRGB(Color=<rgba 0.1724 0.6122 1 1>)
     "Float (Alpha)"=&FloatFromCurve(
      FloatCurve=DataInterface() {propertyValues {Curve {keys [
          <at 0 0>
          {arriveTangent=1.25, leaveTangent=1.25, time=0.2028, value=0.9955}
          {arriveTangent=-1.2487, leaveTangent=-1.2487, time=0.8, value=1}
          <at 1 0>
         ]}}}
      Scale Curve=0.25
     )
    )) {isSetParameters=true, module=SetVariables_#2}
   /SetVariables(Particles.SizeGlow:Vector2f=&Multiply_Vector2D_ByFloat(Float=0.5, Vector2D=$Particles.SpriteSize)) {isSetParameters=true, module=SetVariables_#3}
   &InheritSourceMovement(Applied Position Scale=<0.6 0.6 0.6>, ApplyForce=false, ApplyPosition=true, ApplyPositionFalloff=false)
   &SolveForcesAndVelocity(Clamp Velocity=false, Limit Acceleration=false)
  ]
  rendererClasses [&NiagaraSpriteRendererProperties]
  renderers [
   {rendererClass=&NiagaraSpriteRendererProperties}
   {rendererClass=&NiagaraSpriteRendererProperties, rendererIndex=1}
  ]
 }
}
summary {
 systemName=NS_Emote_MajorSpeech_TranformFlare
 emitters [
  {emitterName=Flare, simTarget=GPUComputeSim, rendererClasses [&NiagaraSpriteRendererProperties]}
  {emitterName=Flare001, simTarget=GPUComputeSim, rendererClasses [&NiagaraSpriteRendererProperties]}
 ]
 userVariables [
  User.bIsFrontEnd {type=NiagaraBool}
  User.bIsFrontEndPreview {type=NiagaraBool}
 ]
}
