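# Niagara system for the MajorSpeech emote's icicle effect: a column of icy meshes with rising airflow, cloud puffs, ground smoke, a flare and flying ice.
# A Main particle writes per-layer sizes, colors, positions and material vectors for spline glow, top and bottom layers, plus light attributes zeroed in front-end preview.
# GPU shard emitters randomly pick sprite or mesh via visibility tags and scatter them with curl noise.
# Time-of-day, front-end and LOD-distance functions tune brightness, opacity and size.
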NiagaraSystem /BRCosmetics/Effects/Fort_Effects/Emotes/MajorSpeech/NS_Emote_MajorSpeech_icicle
dependencies {
 usedDataInterfaces [
  &NiagaraDataInterfaceCurve
  &NiagaraDataInterfaceColorCurve
  &NiagaraDataInterfaceMeshRendererInfo
  &NiagaraDataInterfaceCamera
  &NiagaraDataInterfaceSpriteRendererInfo
  &NiagaraDataInterfaceArrayFloat
  &NiagaraDataInterfaceArrayFloat3
  &NiagaraDataInterfaceVectorCurve
 ]
 usedDynamicInputs []
 usedModules [&SystemState]
}
emitters {
 Airflow {
  emitterUpdateScript [
   &EmitterState(Life Cycle Mode=::NewEnumerator1, Loop Behavior=::NewEnumerator1, Loop Duration=0.65, UseLoopDelay=true)
   &SpawnBurst_Instantaneous(Spawn Count=5, Use Spawn Probability=false, UseLoopCountLimit=false)
   &SpawnRate(SpawnRate=20, Use Spawn Probability=false)
  ]
  particleSpawnScript [
   &InitializeParticle(
    Color=<rgba 0.5 0.7865 1 1>
    Lifetime=0.8
    Mesh Scale=<0.95 0.95 2.5>
    Mesh Scale Mode=::NewEnumerator4
    Position Offset=<0 0 20>
    UsePositionOffset=true
    Write Material Random=true
   )
   &ShapeLocation(
    Handle Radius=0
    Large Radius=20
    MaskSpawnGroup=false
    Non Uniform Scale=<1 1 2>
    Rotation Mode=::NewEnumerator1
    Shape Primitive=::NewEnumerator3
    Torus Distribution Mode:ENiagaraTorusDistributionMode=::NewEnumerator1
    U Position=&ReturnNormalizedExecIndex()
    UseOwnerScale=true
   )
   &InitialMeshOrientation(
    Mesh Orientation Mode=::NewEnumerator2
    Rotation=&RandomRangeVector(Minimum=<0 0 0>, Maximum=<0 0 1>)
    Use Rotation Vector=true
   )
   &InitialMeshOrientation(Mesh Orientation Mode=::NewEnumerator2, Use Rotation Vector=true) {module=InitialMeshOrientation001}
   &AddVelocity(Velocity=<0 0 100>)
  ]
  particleUpdateScript [
   &ParticleState()
   &UpdateMeshOrientation(Rotation Rate=0.3, Rotation Vector=<0 0 1>)
   &ScaleMeshSize(Scale Factor=&Multiply_VectorByFloat(
     Vector=&VectorFromCurve(
      VectorCurve=DataInterface() {propertyValues {XCurve {keys [<at 0 0.8>,<at 1 1.2>]}, YCurve {keys [<at 0 0.8>,<at 1 1.2>]}, ZCurve {keys [<at 0 1>,<at 1 1>]}}}
      Scale Curve=&Multiply_VectorByFloat(Vector=<1 1 1>, Float=1)
     )
     Float=1
    ))
   &DynamicMaterialParameters(
    Index 0 Param 1=&FloatFromCurve(FloatCurve=DataInterface() {propertyValues {Curve {keys [<at 0 0.1>,<at 1 0.5>]}}})
    Index 0 Param 2=&RandomRangeFloat(Minimum=50, Maximum=80)
    Module.Index0Param0Write=true
    Module.Index0Param1Write=true
    Module.Index0Param2Write=true
    Module.Index0Param3Write=true
   )
   &ScaleColor(
    Scale Alpha=2
    Scale RGB=&Multiply_VectorByFloat(
     Vector=&MakeVectorFromLinearColorRGB(Color=&FNTimeOfDayColor(Night Color=<rgba 0.5299 0.8132 1 1>, Morning Color=<rgba 1 1 1 1>, Day Color=<rgba 1 1 1 1>, Evening Color=<rgba 1 1 1 1>))
     Float=&FloatFromCurve(FloatCurve=DataInterface() {propertyValues {Curve {keys [
          <at 0 0 RCIM_Cubic>
          <at 0.5 1 RCIM_Cubic>
          {arriveTangent=-0.6427, interpMode=RCIM_Cubic, leaveTangent=-0.6427, tangentMode=RCTM_User, time=1, value=0}
         ]}}})
    )
    ScaleA=true
    ScaleRGB=true
   )
   &SolveForcesAndVelocity(Clamp Velocity=false, Limit Acceleration=false)
  ]
  rendererClasses [&NiagaraMeshRendererProperties]
  renderers [{rendererClass=&NiagaraMeshRendererProperties, rendererIndex=0}]
 }
 Cloud_Top {
  emitterUpdateScript [
   &EmitterState(Inactive Response=::NewEnumerator1, Life Cycle Mode=::NewEnumerator1, Loop Behavior=::NewEnumerator1, Loop Delay=0.1, UseLoopDelay=true)
   &SpawnBurst_Instantaneous(Spawn Count=5, Use Spawn Probability=false, UseLoopCountLimit=false)
   &SpawnRate(
    SpawnRate=&Multiply_Float(
     A=25
     B=&FloatFromCurve(FloatCurve=DataInterface() {propertyValues {Curve {keys [<at 0 1>,<at 0.4 1>,<at 0.5 0>]}}}, CurveIndex=$Emitter.LoopedAge)
    )
    Use Spawn Probability=true
   )
  ]
  particleSpawnScript [
   &InitializeParticle(
    Lifetime Max=0.8
    Lifetime Min=0.6
    Lifetime Mode=::NewEnumerator1
    Mesh Scale Mode=::NewEnumerator4
    UsePositionOffset=false
    Write Material Random=true
   )
   &InitialMeshOrientation(
    Mesh Orientation Mode=::NewEnumerator2
    Rotation=&RandomRangeVector(Minimum=<0 0 0>, Maximum=<1 1 1>)
    Use Rotation Vector=true
   )
   &CameraOffset()
   &ShapeLocation(
    Handle Radius=10
    Large Radius=30
    MaskSpawnGroup=false
    Offset=<0 0 250>
    Offset Mode=::NewEnumerator0
    Override Seed=false
    Shape Primitive=::NewEnumerator3
    UseOwnerScale=true
   )
   &AddVelocity(Module.Use Falloff Distance=false, Origin Offset=<0 0 230>, Override Seed=false, Velocity Mode=::NewEnumerator1, Velocity Speed=50)
   &AddVelocity(Velocity=<0 0 2>) {module=AddVelocity001}
  ]
  particleUpdateScript [
   &ParticleState()
   &UpdateMeshOrientation(Rotation Rate=0.1)
   &Drag(Use Linear Drag=true, Use Rotational Drag=true)
   &SolveForcesAndVelocity(Clamp Velocity=false, Limit Acceleration=false)
   &ScaleMeshSize(Scale Factor=&VectorFromFloat(Value=&FloatFromCurve(
      FloatCurve=DataInterface() {propertyValues {Curve {keys [
          {arriveTangent=0.843, interpMode=RCIM_Cubic, leaveTangent=0.843, tangentMode=RCTM_User, time=0, value=0.8}
          {arriveTangent=0.2098, interpMode=RCIM_Cubic, leaveTangent=0.2098, tangentMode=RCTM_User, time=1, value=1.6}
         ]}}}
      Scale Curve=0.9
     )))
   &DynamicMaterialParameters(
    Index 0 Param 1=&FloatFromCurve(
     FloatCurve=DataInterface() {propertyValues {Curve {keys [
         {interpMode=RCIM_Cubic, tangentMode=RCTM_User, time=0, value=0}
         <at 0.2 0.55>
         {arriveTangent=-1.0223, interpMode=RCIM_Cubic, leaveTangent=-1.0223, tangentMode=RCTM_User, time=1, value=1}
        ]}}}
     Scale Curve=&RandomRangeFloat(Minimum=0.7, Maximum=1.5)
    )
    Index 0 Param 3=1
    Index 0 Param 4=&FloatFromCurve(FloatCurve=DataInterface() {propertyValues {Curve {keys [
         {arriveTangent=-0.9471, interpMode=RCIM_Cubic, leaveTangent=-0.9471, tangentMode=RCTM_User, time=0, value=0.2}
         {arriveTangent=-1.9736, interpMode=RCIM_Cubic, leaveTangent=-1.9736, tangentMode=RCTM_User, time=0.1, value=0.75}
         {arriveTangent=-0.1255, interpMode=RCIM_Cubic, leaveTangent=-0.1255, tangentMode=RCTM_User, time=1, value=0}
        ]}}})
    Module.Index0Param0Write=true
    Module.Index0Param1Write=true
    Module.Index0Param2Write=true
    Module.Index0Param3Write=true
   )
   &ScaleColor(
    Scale Alpha=&FloatFromCurve(FloatCurve=DataInterface() {propertyValues {Curve {keys [{arriveTangent=2.8927, interpMode=RCIM_Cubic, time=0, value=0}, <at 0.4 1 RCIM_Cubic>, <at 1 0 RCIM_Cubic>]}}})
    Scale RGB=&MakeVectorFromLinearColorRGB(Color=&ColorFromCurve(ColorCurve=DataInterface() {propertyValues {
        AlphaCurve {keys []}
        BlueCurve {keys [{arriveTangent=-331.6383, interpMode=RCIM_Cubic, time=0, value=1}]}
        GreenCurve {keys [{arriveTangent=-87.3695, interpMode=RCIM_Cubic, time=0, value=1}]}
        RedCurve {keys [<at 0 1 RCIM_Cubic>]}
       }}))
    ScaleA=true
    ScaleRGB=true
   )
  ]
  rendererClasses [&NiagaraMeshRendererProperties]
  renderers [{rendererClass=&NiagaraMeshRendererProperties, rendererIndex=0}]
 }
 Cloud_Top001 {
  emitterUpdateScript [
   &EmitterState(
    Inactive Response=::NewEnumerator1
    Life Cycle Mode=::NewEnumerator1
    Loop Behavior=::NewEnumerator1
    Loop Delay=0.05
    Loop Duration=1.5
    UseLoopDelay=true
   )
   &SpawnBurst_Instantaneous(Spawn Count=6, Use Spawn Probability=false, UseLoopCountLimit=false)
   &SpawnRate(
    SpawnRate=&Multiply_Float(
     A=20
     B=&FloatFromCurve(FloatCurve=DataInterface() {propertyValues {Curve {keys [<at 0 1>,<at 0.3 1>,<at 0.4 0>]}}}, CurveIndex=$Emitter.LoopedAge)
    )
    Use Spawn Probability=true
   )
  ]
  particleSpawnScript [
   &PartitionParticles(Number Of Partitions=4)
   &InitializeParticle(
    Lifetime=&SelectFloatFromArray(Direct Array Index=$Particles.PartitionParticles.Partition, Float Selection Array=DataInterface() {propertyValues {FloatData [1,1,1,1]}})
    Mesh Scale=&SelectVectorFromArray(
     Direct Array Index=$Particles.PartitionParticles.Partition
     Vector Selection Array=DataInterface() {propertyValues {FloatData [<1 1 1>, <0.8 0.8 0.8>, <0.8 0.8 0.8>, <1.3 1.3 1.3>]}}
    )
    Mesh Scale Mode=::NewEnumerator4
    UsePositionOffset=false
    Write Material Random=true
   )
   &InitialMeshOrientation(
    Mesh Orientation Mode=::NewEnumerator2
    Rotation=&RandomRangeVector(Minimum=<0 0 0>, Maximum=<1 1 1>)
    Use Rotation Vector=true
   )
   &ShapeLocation(
    MaskSpawnGroup=false
    Offset=&SelectVectorFromArray(
     Direct Array Index=$Particles.PartitionParticles.Partition
     Vector Selection Array=DataInterface() {propertyValues {FloatData [<0 0 200>, <20 60 65>, <-60 -10 135>, <0 45 130>]}}
    )
    Offset Mode=::NewEnumerator0
    Override Seed=false
    Sphere Radius=5
    UseOwnerScale=true
   )
   &CameraOffset(Camera Offset Amount=20)
  ]
  particleUpdateScript [
   &ParticleState()
   &UpdateMeshOrientation(Rotation Rate=0.2)
   &Drag(Use Linear Drag=true, Use Rotational Drag=true)
   &SolveForcesAndVelocity(Clamp Velocity=false, Limit Acceleration=false)
   &ScaleMeshSize(Scale Factor=&VectorFromFloat(Value=&FloatFromCurve(FloatCurve=DataInterface() {propertyValues {Curve {keys [
          {arriveTangent=0.843, interpMode=RCIM_Cubic, leaveTangent=0.843, tangentMode=RCTM_User, time=0, value=0.8}
          {arriveTangent=0.2098, interpMode=RCIM_Cubic, leaveTangent=0.2098, tangentMode=RCTM_User, time=1, value=1.6}
         ]}}})))
   &DynamicMaterialParameters(
    Index 0 Param 1=&FloatFromCurve(
     FloatCurve=DataInterface() {propertyValues {Curve {keys [
         {interpMode=RCIM_Cubic, tangentMode=RCTM_User, time=0, value=0}
         <at 0.2 0.55>
         {arriveTangent=-1.0223, interpMode=RCIM_Cubic, leaveTangent=-1.0223, tangentMode=RCTM_User, time=1, value=1}
        ]}}}
     Scale Curve=&RandomRangeFloat(Minimum=0.7, Maximum=1.5)
    )
    Index 0 Param 3=1
    Index 0 Param 4=&FloatFromCurve(FloatCurve=DataInterface() {propertyValues {Curve {keys [
         {arriveTangent=-0.9471, interpMode=RCIM_Cubic, leaveTangent=-0.9471, tangentMode=RCTM_User, time=0, value=0.2}
         {arriveTangent=-1.9736, interpMode=RCIM_Cubic, leaveTangent=-1.9736, tangentMode=RCTM_User, time=0.1, value=0.75}
         {arriveTangent=-0.1255, interpMode=RCIM_Cubic, leaveTangent=-0.1255, tangentMode=RCTM_User, time=1, value=0}
        ]}}})
    Module.Index0Param0Write=true
    Module.Index0Param1Write=true
    Module.Index0Param2Write=true
    Module.Index0Param3Write=true
   )
   &ScaleColor(
    Scale Alpha=&FloatFromCurve(
     FloatCurve=DataInterface() {propertyValues {Curve {keys [
         {arriveTangent=2.8927, interpMode=RCIM_Cubic, time=0, value=0}
         {arriveTangent=-0.0062, interpMode=RCIM_Cubic, leaveTangent=-0.0062, tangentMode=RCTM_User, time=0.1, value=1}
         {arriveTangent=-0.0374, interpMode=RCIM_Cubic, leaveTangent=-0.0374, tangentMode=RCTM_User, time=0.8, value=1}
         <at 1 0 RCIM_Cubic>
        ]}}}
     Scale Curve=&SelectFloatFromArray(Direct Array Index=$Particles.PartitionParticles.Partition, Float Selection Array=DataInterface() {propertyValues {FloatData [2,1.2,1.2]}})
    )
    Scale RGB=&MakeVectorFromLinearColorRGB(Color=&ColorFromCurve(ColorCurve=DataInterface() {propertyValues {
        AlphaCurve {keys []}
        BlueCurve {keys [{arriveTangent=-331.6383, interpMode=RCIM_Cubic, time=0, value=1}]}
        GreenCurve {keys [{arriveTangent=-87.3695, interpMode=RCIM_Cubic, time=0, value=1}]}
        RedCurve {keys [<at 0 1 RCIM_Cubic>]}
       }}))
    ScaleA=true
    ScaleRGB=true
   )
  ]
  rendererClasses [&NiagaraMeshRendererProperties]
  renderers [{rendererClass=&NiagaraMeshRendererProperties, rendererIndex=0}]
 }
 Flare {
  emitterUpdateScript [
   &EmitterState(Life Cycle Mode=::NewEnumerator1, Loop Behavior=::NewEnumerator1, Loop Duration=1, UseLoopDelay=false)
   &SpawnBurst_Instantaneous(Spawn Count=1, Use Spawn Probability=false)
  ]
  particleSpawnScript [
   &InitializeParticle(
    Color=<rgba 0 1.3333 2 1>
    Lifetime=0.25
    Position Offset=<0 0 20>
    Sprite Size=<500 800>
    Sprite Size Mode=::NewEnumerator4
    UsePositionOffset=true
    Write Material Random=true
   )
   &DynamicMaterialParameters(
    Index 0 Param 1=&RandomRangeFloat(Minimum=0.4, Maximum=0.6)
    Index 0 Param 2=1
    Index 0 Param 3=1
    Index 0 Param 4=1
    Module.Param0WriteEnabled=true
    Module.Param1WriteEnabled=true
    Module.Param2WriteEnabled=true
    Module.Param3WriteEnabled=true
   )
   &CameraOffset(Camera Offset Amount=60)
   &AddVelocity(Velocity=<0 0 800>)
  ]
  particleUpdateScript [
   &ParticleState()
   &ScaleColor(
    Scale Alpha=&Multiply_Float(
     A=&FNTimeOfDayScalar(Night Scale=0.7, Morning Scale=0.3, Evening Scale=0.4)
     B=&FloatFromCurve(FloatCurve=DataInterface() {propertyValues {Curve {keys [
          <at 0 0 RCIM_Cubic>
          {arriveTangent=-0.0814, interpMode=RCIM_Cubic, leaveTangent=-0.0814, tangentMode=RCTM_User, time=0.15, value=1}
          {arriveTangent=-0.0472, interpMode=RCIM_Cubic, leaveTangent=-0.0472, tangentMode=RCTM_User, time=0.6, value=1}
          <at 1 0 RCIM_Cubic>
         ]}}})
    )
    Scale RGB=&Multiply_VectorByFloat(Vector=<1 1 1>, Float=&FN_FrontEndFloatModifier(Gameplay=1.5, Front End=1))
    ScaleA=true
    ScaleRGB=true
   )
   &ScaleSpriteSize(Scale Factor=&Vector2DFromFloat(Value=&FloatFromCurve(FloatCurve=DataInterface() {propertyValues {Curve {keys [<at 0 0.5 RCIM_Cubic>,<at 1 1.3>]}}})))
   &Drag(Use Linear Drag=true, Use Rotational Drag=true)
   &SolveForcesAndVelocity(Clamp Velocity=false, Limit Acceleration=false)
  ]
  rendererClasses [&NiagaraSpriteRendererProperties]
  renderers [{rendererClass=&NiagaraSpriteRendererProperties}]
 }
 Ground_MeshSmoke {
  emitterUpdateScript [
   &EmitterState(Life Cycle Mode=::NewEnumerator1, Loop Behavior=::NewEnumerator1, UseLoopDelay=false)
   &SpawnBurst_Instantaneous(Use Spawn Probability=false, UseLoopCountLimit=false)
   &SpawnRate(
    SpawnRate=&Multiply_Float(
     A=10
     B=&FloatFromCurve(FloatCurve=DataInterface() {propertyValues {Curve {keys [<at 0 1>,<at 0.7 1>,<at 0.9 0>]}}}, CurveIndex=$Emitter.LoopedAge)
    )
    Use Spawn Probability=false
   )
  ]
  particleSpawnScript [
   &InitializeParticle(
    Color Mode=::NewEnumerator0
    Lifetime=0.4
    Mesh Scale Mode=::NewEnumerator4
    Position Offset=<0 0 20>
    UsePositionOffset=true
    Write Material Random=true
   )
   &InitialMeshOrientation(
    Mesh Orientation Mode=::NewEnumerator2
    Rotation=&RandomRangeVector(Minimum=<0 0 0>, Maximum=<0 0 1>)
    Use Rotation Vector=true
   )
  ]
  particleUpdateScript [
   &ParticleState()
   &ScaleMeshSize(Scale Factor=&VectorFromFloat(Value=&FloatFromCurve(
      FloatCurve=DataInterface() {propertyValues {Curve {keys [
          {arriveTangent=0.583, interpMode=RCIM_Cubic, leaveTangent=0.583, tangentMode=RCTM_User, time=0, value=0.6}
          {arriveTangent=0.0824, interpMode=RCIM_Cubic, leaveTangent=0.0824, tangentMode=RCTM_User, time=1, value=1}
         ]}}}
      Scale Curve=&RandomRangeFloat(Minimum=1.2, Maximum=1.3)
     )))
   &ScaleColor(
    Scale Alpha=&FloatFromCurve(FloatCurve=DataInterface() {propertyValues {Curve {keys [
         {arriveTangent=3.1696, interpMode=RCIM_Cubic, leaveTangent=3.1696, tangentMode=RCTM_User, time=0, value=0}
         <at 0.3 1 RCIM_Cubic>
         <at 1 0 RCIM_Cubic>
        ]}}})
    Scale RGB=&MakeVectorFromLinearColorRGB(Color=&ColorFromCurve(ColorCurve=DataInterface() {propertyValues {
        AlphaCurve {keys []}
        BlueCurve {keys [{arriveTangent=-331.6383, interpMode=RCIM_Cubic, time=0, value=1}]}
        GreenCurve {keys [{arriveTangent=-87.3695, interpMode=RCIM_Cubic, time=0, value=1}]}
        RedCurve {keys [<at 0 1 RCIM_Cubic>]}
       }}))
    ScaleA=true
    ScaleRGB=true
   )
   &DynamicMaterialParameters(
    Index 0 Param 1=&FloatFromCurve(
     FloatCurve=DataInterface() {propertyValues {Curve {keys [
         <at 0 0 RCIM_Cubic>
         {arriveTangent=1, interpMode=RCIM_Cubic, leaveTangent=1, time=0.2, value=0.55}
         {interpMode=RCIM_Cubic, leaveTangent=-1.0223, time=1, value=1}
        ]}}}
     Scale Curve=&RandomRangeFloat(Minimum=1, Maximum=1.5)
    )
    Index 0 Param 3=1
    Index 0 Param 4=&FloatFromCurve(FloatCurve=DataInterface() {propertyValues {Curve {keys [
         {arriveTangent=-0.0115, interpMode=RCIM_Cubic, leaveTangent=-0.0115, tangentMode=RCTM_User, time=0, value=1}
         {arriveTangent=-1, interpMode=RCIM_Cubic, leaveTangent=-1, time=0.3, value=0.75}
         {arriveTangent=-1.7045, interpMode=RCIM_Cubic, leaveTangent=-1.7045, tangentMode=RCTM_User, time=0.4, value=0.6}
         {arriveTangent=-0.1255, interpMode=RCIM_Cubic, leaveTangent=-0.1255, tangentMode=RCTM_User, time=1, value=0}
        ]}}})
    Module.Index0Param0Write=true
    Module.Index0Param1Write=true
    Module.Index0Param2Write=true
    Module.Index0Param3Write=true
   ) {enabled=false}
  ]
  rendererClasses [&NiagaraMeshRendererProperties]
  renderers [{rendererClass=&NiagaraMeshRendererProperties, rendererIndex=0}]
 }
 Ice {
  simTarget=GPUComputeSim
  emitterUpdateScript [
   &EmitterState(Inactive Response=::NewEnumerator1, Life Cycle Mode=::NewEnumerator1, Loop Behavior=::NewEnumerator1, UseLoopDelay=false)
   &SpawnBurst_Instantaneous(Spawn Count=20, Use Spawn Probability=false, UseLoopCountLimit=false)
   &SpawnRate(
    Spawn Probability=&FloatFromCurve(
     FloatCurve=DataInterface() {propertyValues {Curve {keys [
         <at 0 1 RCIM_Cubic>
         {arriveTangent=0.0277, interpMode=RCIM_Cubic, leaveTangent=-2.0392, tangentMode=RCTM_Break, time=0.3, value=1}
         <at 1 0 RCIM_Cubic>
        ]}}}
     CurveIndex=$Emitter.Age
    )
    SpawnRate=50
    Use Spawn Probability=true
   )
  ]
  particleSpawnScript [
   &InitializeParticle(
    Lifetime=0.45
    Mesh Scale Max=<0.12 0.12 0.3>
    Mesh Scale Min=<0.07 0.07 0.2>
    Mesh Scale Mode=::NewEnumerator2
    UsePositionOffset=false
    Write Material Random=true
   )
   &ShapeLocation(
    Cylinder Height=200
    Cylinder Radius=120
    MaskSpawnGroup=false
    Override Seed=false
    Shape Primitive=::NewEnumerator1
    Surface Only=true
    Surface Only Band Thickness=0.5
    Use Endcaps In Surface Only Mode=false
    UseOwnerScale=true
   )
   &AddVelocity(Velocity=&RandomRangeVector(Minimum=<0 0 400>, Maximum=<0 0 600>))
   &InitialMeshOrientation(
    Mesh Orientation Mode=::NewEnumerator2
    Rotation=&RandomRangeVector(Minimum=<0 0 0>, Maximum=<0 0 1>)
    Use Rotation Vector=true
   )
  ]
  particleUpdateScript [
   &ParticleState()
   &UpdateMeshOrientation(Rotation Rate=2, Rotation Vector=<0 0 1>)
   &SolveForcesAndVelocity(Clamp Velocity=false, Limit Acceleration=false)
  ]
  rendererClasses [&NiagaraMeshRendererProperties]
  renderers [{rendererClass=&NiagaraMeshRendererProperties, rendererIndex=0}]
 }
 Ice_Shards {
  simTarget=GPUComputeSim
  emitterUpdateScript [
   &EmitterState(Inactive Response=::NewEnumerator1, Life Cycle Mode=::NewEnumerator1, Loop Behavior=::NewEnumerator1, UseLoopDelay=false)
   &SpawnBurst_Instantaneous(Spawn Count=100, Use Spawn Probability=false, UseLoopCountLimit=true)
   &SpawnRate(
    Spawn Probability=&FloatFromCurve(
     FloatCurve=DataInterface() {propertyValues {Curve {keys [
         <at 0 1 RCIM_Cubic>
         {arriveTangent=0.0277, interpMode=RCIM_Cubic, leaveTangent=-2.0392, tangentMode=RCTM_Break, time=0.3, value=1}
         <at 1 0 RCIM_Cubic>
        ]}}}
     CurveIndex=$Emitter.Age
    )
    SpawnRate=800
    Use Spawn Probability=true
   )
  ]
  particleSpawnScript [
   /SetVariables(
    Particles.Particle_Depth:NiagaraFloat=&FloatFromCurve(
     FloatCurve=DataInterface() {propertyValues {Curve {keys [
         {arriveTangent=1.2753, interpMode=RCIM_Cubic, leaveTangent=1.2753, tangentMode=RCTM_User, time=0, value=0}
         {arriveTangent=0.204, interpMode=RCIM_Cubic, leaveTangent=0.204, tangentMode=RCTM_User, time=0.9, value=1}
        ]}}}
     CurveIndex=&RandomRangeFloat(Minimum=0, Maximum=1)
    )
    Particles.Particle_Height:NiagaraFloat=&FloatFromCurve(
     FloatCurve=DataInterface() {propertyValues {Curve {keys [
         {arriveTangent=1.9477, interpMode=RCIM_Cubic, leaveTangent=1.9477, tangentMode=RCTM_User, time=0, value=0}
         {arriveTangent=0.1423, interpMode=RCIM_Cubic, leaveTangent=0.1423, tangentMode=RCTM_User, time=0.5, value=0.5}
         {arriveTangent=1.4427, interpMode=RCIM_Cubic, leaveTangent=1.4427, tangentMode=RCTM_User, time=1, value=1}
        ]}}}
     CurveIndex=&RandomRangeFloat(Minimum=0, Maximum=1)
    )
    Particles.Particle_Scale:NiagaraFloat=&RandomRangeFloat(Minimum=0, Maximum=1)
   ) {isSetParameters=true, module=SetVariables_#1}
   &InitializeParticle(
    Color Maximum=<rgba 0.9 0.9833 1 1>
    Color Minimum=<rgba 0.6 0.8667 1 1>
    Color Mode=::NewEnumerator2
    Lifetime=0.45
    Mass Mode=::NewEnumerator3
    Mesh Renderer Array Visibility Mode=::NewEnumerator3
    Mesh Renderer Info=DataInterface() {propertyValues {MeshRenderer=&self:Ice_Shards_0.NiagaraMeshRendererProperties_0}}
    Mesh Scale Mode=::NewEnumerator1
    Mesh Uniform Scale Max=0.35
    Mesh Uniform Scale Min=0.18
    Sprite Rotation Mode=::NewEnumerator2
    Sprite Size Mode=::NewEnumerator1
    Uniform Sprite Size Min=3
    UsePositionOffset=false
    Write Material Random=true
   )
   &ShapeLocation(
    Cylinder Height=200
    Cylinder Radius=120
    MaskSpawnGroup=false
    Override Seed=false
    Shape Primitive=::NewEnumerator1
    Surface Only=false
    UseOwnerScale=true
   )
   &CameraOffset(Camera Offset Amount=6)
   &InitialMeshOrientation(
    Mesh Orientation Mode=::NewEnumerator2
    Rotation=&RandomRangeVector(Minimum=<0 0 0>, Maximum=<1 1 1>)
    Use Rotation Vector=true
   )
   /SetVariables(Particles.RotationDirection:Vector3f=&Lerp_Vector(
     A=&RandomRangeVector(Minimum=<-0.6 -0.6 1>, Maximum=<0.8 0.8 1>)
     B=&RandomRangeVector(Minimum=<-0.2 -0.2 1>, Maximum=<0.3 0.3 1>)
     Alpha=&FloatFromCurve(
      FloatCurve=DataInterface() {propertyValues {Curve {keys [
          {interpMode=RCIM_Cubic, tangentMode=RCTM_User, time=0, value=0}
          {arriveTangent=0.5342, interpMode=RCIM_Cubic, leaveTangent=0.5342, tangentMode=RCTM_User, time=1, value=1}
         ]}}}
      CurveIndex=$Particles.Particle_Depth
     )
    )) {isSetParameters=true, module=SetVariables_#2}
   /SetVariables(Particles.isSnowFlake:NiagaraBool=&RandomBool()) {isSetParameters=true, module=SetVariables_#3}
   /SetVariables(Particles.VisibilityTag=&MakeCustomIntFromBool(Bool=$Particles.isSnowFlake, True INT=0, False INT=1)) {isSetParameters=true, module=SetVariables_#4}
   &DynamicMaterialParameters(
    Index 0 Param 1=&RandomRangeFloat(Minimum=1, Maximum=2)
    Module.Index0Param0Write=true
    Module.Index0Param1Write=false
    Module.Index0Param2Write=false
    Module.Index0Param3Write=false
   )
   &AddVelocity(Velocity=&RandomRangeVector(Minimum=<0 0 400>, Maximum=<0 0 600>))
  ]
  particleUpdateScript [
   &ParticleState()
   &SpriteRotationRate(Rotation Rate=&RandomRangeFloat(Minimum=300, Maximum=500))
   &UpdateMeshOrientation(
    Rotation Rate=&Multiply_Float(
     A=&Lerp_Float(A=2, B=1, Alpha=$Particles.Particle_Depth)
     B=&Lerp_Float(A=3, B=0.5, Alpha=$Particles.Particle_Scale)
    )
    Rotation Vector=&RandomRangeVector(Minimum=<-1 -1 -1>, Maximum=<1 1 1>)
   )
   &ScaleMeshSize(Scale Factor=&VectorFromFloat(Value=&FloatFromCurve(FloatCurve=DataInterface() {propertyValues {Curve {keys [
          {interpMode=RCIM_Cubic, tangentMode=RCTM_User, time=0, value=0}
          {interpMode=RCIM_Cubic, tangentMode=RCTM_User, time=0.1, value=1}
          {arriveTangent=-3.406, interpMode=RCIM_Cubic, leaveTangent=-3.406, tangentMode=RCTM_User, time=1, value=0}
         ]}}})))
   &ScaleSpriteSize(Uniform Curve Sprite Scale=DataInterface() {propertyValues {Curve {keys [{interpMode=RCIM_Cubic, tangentMode=RCTM_User, time=0, value=0.6}, {interpMode=RCIM_Cubic, tangentMode=RCTM_User, time=0.2, value=1}]}}})
   &SpriteFacingAndAlignment(
    Sprite Facing=&Lerp_Vector(
     A=&RandomRangeVector(Minimum=<-1 -1 -1>, Maximum=<1 1 1>)
     B=&ReturnCameraVector(Camera Vectors=::NewEnumerator1, Coordinate Space=::Simulation)
     Alpha=&FloatFromCurve(
      FloatCurve=DataInterface() {propertyValues {Curve {keys [
          {arriveTangent=2.0555, interpMode=RCIM_Cubic, leaveTangent=2.0555, tangentMode=RCTM_User, time=0, value=0}
          {interpMode=RCIM_Cubic, tangentMode=RCTM_User, time=1, value=1}
         ]}}}
      CurveIndex=&RandomRangeFloat(Minimum=0, Maximum=1)
     )
    )
    WriteAlignment=false
    WriteFacing=true
   )
   &ScaleColor(
    Scale Alpha=&RandomRangeFloat(Minimum=0.15, Maximum=0.5)
    Scale RGB=&MakeVectorFromLinearColorRGB(Color=<rgba 0.5625 0.896 1 1>)
    ScaleA=true
    ScaleRGB=true
   )
   &CurlNoiseForce(Bias Noise=false, Noise Strength=600, Pan Noise=false)
   &Drag(Drag=0.5, Use Linear Drag=true, Use Rotational Drag=true)
   &SolveForcesAndVelocity(Clamp Velocity=false, Limit Acceleration=false)
   /SetVariables(Particles.Color02:LinearColor=&MakeLinearColorFromVectorAndFloat("Vector (RGB)"=&MakeVectorFromLinearColorRGB(Color=<rgba 0.375 1.1543 1.5 1>), "Float (Alpha)"=1)) {isSetParameters=true, module=SetVariables_#5}
  ]
  rendererClasses [&NiagaraSpriteRendererProperties, &NiagaraMeshRendererProperties]
  renderers [
   {rendererClass=&NiagaraSpriteRendererProperties}
   {rendererClass=&NiagaraMeshRendererProperties, rendererIndex=1}
  ]
 }
 Ice_Shards_Top {
  simTarget=GPUComputeSim
  emitterUpdateScript [
   &EmitterState(Life Cycle Mode=::NewEnumerator1, Loop Behavior=::NewEnumerator1, Loop Duration=0.8, UseLoopDelay=false)
   &SpawnBurst_Instantaneous(Spawn Count=5, Use Spawn Probability=false, UseLoopCountLimit=false)
   &SpawnRate(
    Spawn Probability=&FloatFromCurve(
     FloatCurve=DataInterface() {propertyValues {Curve {keys [
         {interpMode=RCIM_Cubic, tangentMode=RCTM_User, time=0.1, value=0}
         {interpMode=RCIM_Cubic, tangentMode=RCTM_User, time=1, value=1}
         {arriveTangent=0.0061, interpMode=RCIM_Cubic, leaveTangent=-0.1111, tangentMode=RCTM_Break, time=9, value=1}
         <at 10.6 0 RCIM_Cubic>
        ]}}}
     CurveIndex=$Emitter.Age
    )
    SpawnRate=1000
    Use Spawn Probability=false
   )
  ]
  particleSpawnScript [
   /SetVariables(
    Particles.Particle_Depth:NiagaraFloat=&FloatFromCurve(
     FloatCurve=DataInterface() {propertyValues {Curve {keys [
         {arriveTangent=1.2753, interpMode=RCIM_Cubic, leaveTangent=1.2753, tangentMode=RCTM_User, time=0, value=0}
         {arriveTangent=0.204, interpMode=RCIM_Cubic, leaveTangent=0.204, tangentMode=RCTM_User, time=0.9, value=1}
        ]}}}
     CurveIndex=&RandomRangeFloat(Minimum=0, Maximum=1)
    )
    Particles.Particle_Height:NiagaraFloat=&FloatFromCurve(
     FloatCurve=DataInterface() {propertyValues {Curve {keys [
         {arriveTangent=1.9477, interpMode=RCIM_Cubic, leaveTangent=1.9477, tangentMode=RCTM_User, time=0, value=0}
         {arriveTangent=0.1423, interpMode=RCIM_Cubic, leaveTangent=0.1423, tangentMode=RCTM_User, time=0.5, value=0.5}
         {arriveTangent=1.4427, interpMode=RCIM_Cubic, leaveTangent=1.4427, tangentMode=RCTM_User, time=1, value=1}
        ]}}}
     CurveIndex=&RandomRangeFloat(Minimum=0, Maximum=1)
    )
    Particles.Particle_Scale:NiagaraFloat=&RandomRangeFloat(Minimum=0, Maximum=1)
   ) {isSetParameters=true, module=SetVariables_#1}
   &InitializeParticle(
    Color Maximum=<rgba 0.9 0.9833 1 1>
    Color Minimum=<rgba 0.6 0.8667 1 1>
    Color Mode=::NewEnumerator2
    Lifetime=0.22
    Mass Mode=::NewEnumerator3
    Mesh Renderer Array Visibility Mode=::NewEnumerator3
    Mesh Renderer Info=DataInterface() {propertyValues {MeshRenderer=&self:Ice_Shards_Top_0.NiagaraMeshRendererProperties_0}}
    Mesh Scale Mode=::NewEnumerator1
    Mesh Uniform Scale Max=0.25
    Mesh Uniform Scale Min=0.15
    Sprite Rotation Mode=::NewEnumerator2
    Sprite Size Mode=::NewEnumerator1
    Uniform Sprite Size Min=3
    UsePositionOffset=false
    Write Material Random=true
   )
   &ShapeLocation(
    Handle Radius=10
    Large Radius=50
    MaskSpawnGroup=false
    Offset=<0 0 260>
    Offset Mode=::NewEnumerator0
    Override Seed=false
    Shape Primitive=::NewEnumerator3
    UseOwnerScale=true
   )
   &CameraOffset(Camera Offset Amount=6)
   &InitialMeshOrientation(
    Mesh Orientation Mode=::NewEnumerator2
    Rotation=&RandomRangeVector(Minimum=<0 0 0>, Maximum=<1 1 1>)
    Use Rotation Vector=true
   )
   /SetVariables(Particles.RotationDirection:Vector3f=&Lerp_Vector(
     A=&RandomRangeVector(Minimum=<-0.6 -0.6 1>, Maximum=<0.8 0.8 1>)
     B=&RandomRangeVector(Minimum=<-0.2 -0.2 1>, Maximum=<0.3 0.3 1>)
     Alpha=&FloatFromCurve(
      FloatCurve=DataInterface() {propertyValues {Curve {keys [
          {interpMode=RCIM_Cubic, tangentMode=RCTM_User, time=0, value=0}
          {arriveTangent=0.5342, interpMode=RCIM_Cubic, leaveTangent=0.5342, tangentMode=RCTM_User, time=1, value=1}
         ]}}}
      CurveIndex=$Particles.Particle_Depth
     )
    )) {isSetParameters=true, module=SetVariables_#2}
   /SetVariables(Particles.isSnowFlake:NiagaraBool=&RandomBool()) {isSetParameters=true, module=SetVariables_#3}
   /SetVariables(Particles.VisibilityTag=&MakeCustomIntFromBool(Bool=$Particles.isSnowFlake, True INT=0, False INT=1)) {isSetParameters=true, module=SetVariables_#4}
   &DynamicMaterialParameters(
    Index 0 Param 1=&RandomRangeFloat(Minimum=1, Maximum=2)
    Module.Index0Param0Write=true
    Module.Index0Param1Write=false
    Module.Index0Param2Write=false
    Module.Index0Param3Write=false
   )
   &AddVelocity(Module.Use Falloff Distance=false, Origin Offset=<0 0 250>, Override Seed=false, Velocity Mode=::NewEnumerator1, Velocity Speed=300)
  ]
  particleUpdateScript [
   &ParticleState()
   &SpriteRotationRate(Rotation Rate=&RandomRangeFloat(Minimum=50, Maximum=150))
   &UpdateMeshOrientation(
    Rotation Rate=&Multiply_Float(
     A=&Lerp_Float(A=1.5, B=1, Alpha=$Particles.Particle_Depth)
     B=&Lerp_Float(A=0.5, B=0.3, Alpha=$Particles.Particle_Scale)
    )
    Rotation Vector=&RandomRangeVector(Minimum=<-1 -1 -1>, Maximum=<1 1 1>)
   )
   &ScaleMeshSize(Scale Factor=&VectorFromFloat(Value=&FloatFromCurve(
      FloatCurve=DataInterface() {propertyValues {Curve {keys [
          {interpMode=RCIM_Cubic, tangentMode=RCTM_User, time=0, value=0}
          {interpMode=RCIM_Cubic, tangentMode=RCTM_User, time=0.2, value=1}
          {arriveTangent=-3.406, interpMode=RCIM_Cubic, leaveTangent=-3.406, tangentMode=RCTM_User, time=1, value=0}
         ]}}}
      Scale Curve=1.5
     )))
   &ScaleSpriteSize(Uniform Curve Sprite Scale=DataInterface() {propertyValues {Curve {keys [{interpMode=RCIM_Cubic, tangentMode=RCTM_User, time=0, value=0}, {interpMode=RCIM_Cubic, tangentMode=RCTM_User, time=0.1, value=1}]}}})
   &SpriteFacingAndAlignment(
    Sprite Facing=&Lerp_Vector(
     A=&RandomRangeVector(Minimum=<-1 -1 -1>, Maximum=<1 1 1>)
     B=&ReturnCameraVector(Camera Vectors=::NewEnumerator1, Coordinate Space=::Simulation)
     Alpha=&FloatFromCurve(
      FloatCurve=DataInterface() {propertyValues {Curve {keys [
          {arriveTangent=2.0555, interpMode=RCIM_Cubic, leaveTangent=2.0555, tangentMode=RCTM_User, time=0, value=0}
          {interpMode=RCIM_Cubic, tangentMode=RCTM_User, time=1, value=1}
         ]}}}
      CurveIndex=&RandomRangeFloat(Minimum=0, Maximum=1)
     )
    )
    WriteAlignment=false
    WriteFacing=true
   )
   &ScaleColor(
    Scale Alpha=&FloatFromCurve(
     FloatCurve=DataInterface() {propertyValues {Curve {keys [
         <at 0 0>
         {arriveTangent=1.1111, leaveTangent=1.1111, time=0.2, value=1}
         {arriveTangent=-1.1111, leaveTangent=-1.1111, time=0.9, value=1}
         <at 1 0>
        ]}}}
     Scale Curve=&RandomRangeFloat(Minimum=0.1, Maximum=0.5)
    )
    Scale RGB=&MakeVectorFromLinearColorRGB(Color=<rgba 0.5625 0.896 1 1>)
    ScaleA=true
    ScaleRGB=true
   )
   &CurlNoiseForce(Bias Noise=false, Noise Strength=800, Pan Noise=false)
   &GravityForce(Gravity=<0 0 -200>)
   &Drag(Drag=0.5, Use Linear Drag=true, Use Rotational Drag=true)
   &SolveForcesAndVelocity(Clamp Velocity=false, Limit Acceleration=false)
   /SetVariables(Particles.Color02:LinearColor=&MakeLinearColorFromVectorAndFloat("Vector (RGB)"=&MakeVectorFromLinearColorRGB(Color=<rgba 0.375 1.1543 1.5 1>), "Float (Alpha)"=1)) {isSetParameters=true, module=SetVariables_#5}
  ]
  rendererClasses [&NiagaraSpriteRendererProperties, &NiagaraMeshRendererProperties]
  renderers [
   {rendererClass=&NiagaraSpriteRendererProperties}
   {rendererClass=&NiagaraMeshRendererProperties, rendererIndex=1}
  ]
 }
 Main {
  emitterUpdateScript [
   &EmitterState(Inactive Response=::NewEnumerator1, Life Cycle Mode=::NewEnumerator1, Loop Behavior=::NewEnumerator1, UseLoopDelay=false)
   &SpawnBurst_Instantaneous(Use Spawn Probability=false, UseLoopCountLimit=false)
  ]
  particleSpawnScript [
   &InitializeParticle(
    Color Mode=::NewEnumerator0
    Mesh Scale=<1 1 1.2>
    Mesh Scale Mode=::NewEnumerator4
    Position Offset=<0 0 -30>
    UsePositionOffset=true
    Write Material Random=true
   )
   /SetVariables(Particles.CameraOffsetMain=20) {isSetParameters=true}
   /SetVariables(Particles.CameraOffset_SplineGlow=80) {isSetParameters=true}
   /SetVariables(Particles.Cameraoffset02=60) {isSetParameters=true}
  ]
  particleUpdateScript [
   &ParticleState()
   &ScaleMeshSize(Scale Factor=&MakeVector(
     X=&FloatFromCurve(FloatCurve=DataInterface() {propertyValues {Curve {keys [<at 0 1.5>,<at 0.2 1>,<at 1 1>]}}})
     Y=&FloatFromCurve(FloatCurve=DataInterface() {propertyValues {Curve {keys [<at 0 1.5>,<at 0.2 1>,<at 1 1>]}}})
     Z=&FloatFromCurve(FloatCurve=DataInterface() {propertyValues {Curve {keys [<at 0 0.5>,<at 0.13 1>,<at 1 1>]}}})
    ))
   &ScaleColor(
    Scale Alpha=&FloatFromCurve(FloatCurve=DataInterface() {propertyValues {Curve {keys [<at 0 0>,<at 0.1 1>,<at 0.7 1>,<at 1 0>]}}}, Scale Curve=1.5)
    Scale RGB=&MakeVectorFromLinearColorRGB(Color=&ColorFromCurve(ColorCurve=DataInterface() {propertyValues {
        AlphaCurve {keys []}
        BlueCurve {keys [<at 0.4434 1 RCIM_Cubic>,<at 0.5756 3 RCIM_Cubic>]}
        GreenCurve {keys [<at 0.4434 1 RCIM_Cubic>,<at 0.5756 2.6802 RCIM_Cubic>]}
        RedCurve {keys [<at 0.4434 1 RCIM_Cubic>,<at 0.5756 1.8707 RCIM_Cubic>]}
       }}))
    ScaleA=true
    ScaleRGB=true
   )
   /SetVariables(Particles.Dyn_Main:Vector4f=&MakeVector4(
     X=&FloatFromCurve(FloatCurve=DataInterface() {propertyValues {Curve {keys [<at 0 1>,<at 0.3 1>,<at 0.5012 0.6311>,<at 1 0.1>]}}})
     Y=0
     Z=&FloatFromCurve(
      FloatCurve=DataInterface() {propertyValues {Curve {keys [<at 0 0>,<at 1 1>]}}}
      Scale Curve=&FloatFromCurve(FloatCurve=DataInterface() {propertyValues {Curve {keys [
           {arriveTangent=-0.2506, interpMode=RCIM_Cubic, leaveTangent=-0.2506, tangentMode=RCTM_User, time=0, value=3}
           {arriveTangent=-0.7999, interpMode=RCIM_Cubic, leaveTangent=-0.7999, tangentMode=RCTM_User, time=1, value=1.7}
          ]}}})
     )
     W=0
    )) {isSetParameters=true}
   /SetVariables(Particles.MainLightSize:Vector3f=&Multiply_Vector(A=$Particles.Scale, B=<1.3 1.3 1>)) {isSetParameters=true}
   /SetVariables(Particles.SplineGlow_Scale:Vector3f=&Multiply_Vector(A=$Particles.Scale, B=<1 1 4.7>)) {isSetParameters=true}
   /SetVariables(Particles.Color_SplineGlow:LinearColor=&MakeLinearColorFromVectorAndFloat(
     "Vector (RGB)"=&MakeVectorFromLinearColorRGB(Color=&ColorFromCurve(ColorCurve=DataInterface() {propertyValues {
         AlphaCurve {keys []}
         BlueCurve {keys [<at 0.3742 1 RCIM_Cubic>,<at 0.4559 1 RCIM_Cubic>]}
         GreenCurve {keys [<at 0.3742 1 RCIM_Cubic>,<at 0.4559 0.7463 RCIM_Cubic>]}
         RedCurve {keys [<at 0.3742 1 RCIM_Cubic>,<at 0.4559 0.2626 RCIM_Cubic>]}
        }}))
     "Float (Alpha)"=&FloatFromCurve(FloatCurve=DataInterface() {propertyValues {Curve {keys [<at 0.1 0>,<at 0.2 2>,<at 0.2955 1>,<at 0.6 1>,<at 0.7 0.25>,<at 0.9 0>]}}})
    )) {isSetParameters=true}
   /SetVariables(Particles.Color02:LinearColor=&MakeLinearColorFromVectorAndFloat(
     "Vector (RGB)"=&MakeVectorFromLinearColorRGB(Color=<rgba 0.2031 0.7306 1 1>)
     "Float (Alpha)"=&FloatFromCurve(FloatCurve=DataInterface() {propertyValues {Curve {keys [<at 0 0>,<at 0.1 1>,<at 0.7 1>,<at 1 0>]}}})
    )) {isSetParameters=true}
   /SetVariables(Particles.size02:Vector2f=&Vector2DFromFloat(Value=300)) {isSetParameters=true}
   /SetVariables(Particles.Position001:NiagaraPosition=<0 0 -10>) {isSetParameters=true}
   /SetVariables(Particles.Dyn02:Vector4f=&MakeVector4(X=2.5, Y=20, Z=0, W=0)) {isSetParameters=true}
   /SetVariables(Particles.Position_Top:NiagaraPosition=<0 0 210>) {isSetParameters=true}
   /SetVariables(Particles.Size_Top:Vector3f=&Multiply_VectorByFloat(
     Vector=<1.1 1.1 0.6>
     Float=&FN_LOD_Distance_Remap(Start Distance=300, End Distance=400, Minimum Output=0.8, Maximum Output=1, Clamp Output=true)
    )) {isSetParameters=true}
   /SetVariables(Particles.Color_Top:LinearColor=&MakeLinearColorFromVectorAndFloat(
     "Vector (RGB)"=&MakeVectorFromLinearColorRGB(Color=&ColorFromCurve(ColorCurve=DataInterface() {propertyValues {
         AlphaCurve {keys []}
         BlueCurve {keys [<at 0.3742 1.8 RCIM_Cubic>]}
         GreenCurve {keys [<at 0.3742 1.8 RCIM_Cubic>]}
         RedCurve {keys [<at 0.3742 1.8 RCIM_Cubic>]}
        }}))
     "Float (Alpha)"=&FloatFromCurve(
      FloatCurve=DataInterface() {propertyValues {Curve {keys [<at 0.1 0>,<at 0.2 1>,<at 0.6 1>,<at 1 0>]}}}
      Scale Curve=&FN_LOD_Distance_Remap(Start Distance=400, End Distance=500, Minimum Output=0.2, Maximum Output=1, Clamp Output=true)
     )
    )) {isSetParameters=true}
   /SetVariables(Particles.Dyn_Top:Vector4f=&MakeVector4(
     X=&FloatFromCurve(FloatCurve=DataInterface() {propertyValues {Curve {keys [<at 0 1>,<at 0.4 1>,<at 0.6 0.6>,<at 1 0>]}}})
     Y=&FloatFromCurve(FloatCurve=DataInterface() {propertyValues {Curve {keys [<at 0 -1.5>,<at 0.4 -0.5>]}}})
     Z=0
     W=0
    )) {isSetParameters=true}
   /SetVariables(Particles.Size_Bottom:Vector3f=<0.9 0.9 -1.2>) {isSetParameters=true}
   /SetVariables(Particles.Color_Bottom:LinearColor=&MakeLinearColorFromVectorAndFloat(
     "Vector (RGB)"=&MakeVectorFromLinearColorRGB(Color=&ColorFromCurve(ColorCurve=DataInterface() {propertyValues {
         AlphaCurve {keys []}
         BlueCurve {keys [<at 0.3742 0.8698 RCIM_Cubic>,<at 0.6266 1.5>]}
         GreenCurve {keys [<at 0.3742 0.8698 RCIM_Cubic>,<at 0.6266 1.5>]}
         RedCurve {keys [<at 0.3742 0.8698 RCIM_Cubic>,<at 0.6266 1.5>]}
        }}))
     "Float (Alpha)"=&FloatFromCurve(FloatCurve=DataInterface() {propertyValues {Curve {keys [<at 0 1>,<at 0.1 1>,<at 0.55 1>,<at 1 0>]}}}, Scale Curve=1.5)
    )) {isSetParameters=true}
   &FN_Light_Attributes(
    Color=<rgba 0.0143 0.4996 1.2 1>
    Create Color Attribute=true
    Create Exponent Attribute=true
    Create Position Attribute=true
    Create Radius Attribute=true
    Exponent=2
    Position=<0 0 120>
    Radius=&FloatFromCurve(
     FloatCurve=DataInterface() {propertyValues {Curve {keys [<at 0 1>,<at 1 1.2>]}}}
     Scale Curve=&MakeCustomFloatFromBool(Bool=$User.bIsFrontEndPreview, True Float=0, False Float=380)
    )
    Scale Color=&FloatFromCurve(
     FloatCurve=DataInterface() {propertyValues {Curve {keys [
         {arriveTangent=4.1988, time=0, value=0}
         {arriveTangent=-0.0105, leaveTangent=-0.0105, time=0.1, value=1}
         {arriveTangent=-1.7152, leaveTangent=-1.7152, time=0.8, value=1}
         {arriveTangent=-0.9284, leaveTangent=-0.9284, time=1, value=0}
        ]}}}
     Scale Curve=&Multiply_Float(A=&FNTimeOfDayScalar(Night Scale=0.5, Morning Scale=1, Evening Scale=1), B=20)
    )
   )
  ]
  rendererClasses [
   &NiagaraMeshRendererProperties
   &NiagaraSpriteRendererProperties
   &NiagaraLightRendererProperties
  ]
  renderers [
   {rendererClass=&NiagaraMeshRendererProperties, rendererIndex=0}
   {rendererClass=&NiagaraMeshRendererProperties, rendererIndex=1}
   {rendererClass=&NiagaraMeshRendererProperties, rendererIndex=2}
   {rendererClass=&NiagaraSpriteRendererProperties, rendererIndex=3}
   {rendererClass=&NiagaraMeshRendererProperties, rendererIndex=4}
   {rendererClass=&NiagaraMeshRendererProperties, rendererIndex=5}
   {rendererClass=&NiagaraLightRendererProperties, rendererIndex=6}
  ]
 }
 Smoke001 {
  emitterUpdateScript [
   &EmitterState(
    Life Cycle Mode=::NewEnumerator1
    Loop Behavior=::NewEnumerator2
    Loop Count=3
    Loop Duration=0.26
    Recalculate Duration Each Loop=true
    UseLoopDelay=false
   )
   &SpawnBurst_Instantaneous(Spawn Count=8, Use Spawn Probability=false, UseLoopCountLimit=false)
  ]
  particleSpawnScript [
   &InitializeParticle(
    Color Maximum=<rgba 0.5417 0.9013 1 1>
    Color Minimum=<rgba 0.251 0.3859 0.4271 1>
    Color Mode=::NewEnumerator2
    Lifetime Max=0.7
    Lifetime Min=0.5
    Lifetime Mode=::NewEnumerator1
    Sprite Rotation Mode=::NewEnumerator2
    Sprite Size Mode=::NewEnumerator1
    Uniform Sprite Size Max=120
    Uniform Sprite Size Min=110
    UsePositionOffset=false
    Write Material Random=true
   )
   &AddVelocity(Velocity=<0 0 180>) {enabled=false}
   &ShapeLocation(
    Handle Radius=10
    Large Radius=50
    MaskSpawnGroup=false
    Offset=<0 0 20>
    Offset Mode=::NewEnumerator0
    Shape Primitive=::NewEnumerator3
    Torus Distribution Mode:ENiagaraTorusDistributionMode=::NewEnumerator1
    U Position=&ReturnNormalizedExecIndex()
    UseOwnerScale=true
   )
   &AddVelocity(Module.Use Falloff Distance=false, Origin Offset=<0 0 20>, Override Seed=false, Velocity Mode=::NewEnumerator1, Velocity Speed=300) {module=AddVelocity001}
   &CameraOffset(Camera Offset Amount=50)
  ]
  particleUpdateScript [
   &ParticleState()
   &Drag(Use Linear Drag=true, Use Rotational Drag=true)
   &ScaleColor(
    Scale Alpha=&FloatFromCurve(
     FloatCurve=DataInterface() {propertyValues {Curve {keys [
         {interpMode=RCIM_Cubic, tangentMode=RCTM_User, time=0, value=0}
         {interpMode=RCIM_Cubic, tangentMode=RCTM_User, time=0.4, value=1}
         {tangentMode=RCTM_User, time=1, value=0}
        ]}}}
     Scale Curve=1.5
    )
    Scale RGB=&MakeVectorFromLinearColorRGB(Color=&FNTimeOfDayColor(Night Color=<rgba 0.4946 0.772 1 1>, Morning Color=<rgba 1 1 1 1>, Day Color=<rgba 1 1 1 1>, Evening Color=<rgba 1 1 1 1>))
    ScaleA=true
    ScaleRGB=true
   )
   &SpriteRotationRate()
   &ScaleSpriteSize(Scale Factor=&Vector2DFromFloat(Value=&FloatFromCurve(FloatCurve=DataInterface() {propertyValues {Curve {keys [{arriveTangent=0.3685, time=0, value=0.8}, <at 0.3 0.9>, {arriveTangent=0.1633, leaveTangent=0.1633, time=1, value=1}]}}})))
   &SubUVAnimation(
    "Sprite Renderer "=DataInterface() {propertyValues {SpriteRenderer=&self:Smoke001_0.NiagaraSpriteRendererProperties_1}}
    UseEndFrame=false
    UseStartFrame=false
   )
   &DynamicMaterialParameters(
    Index 0 Param 1=1
    Index 0 Param 4=1
    Module.Index0Float0Write=true
    Module.Index0Float1Write=true
    Module.Index0Float2Write=true
    Module.Index0Float3Write=true
   )
   &SolveForcesAndVelocity(Clamp Velocity=false, Limit Acceleration=false)
  ]
  rendererClasses [&NiagaraSpriteRendererProperties]
  renderers [{rendererClass=&NiagaraSpriteRendererProperties}]
 }
 Smoke_Disappear {
  emitterUpdateScript [
   &EmitterState(Life Cycle Mode=::NewEnumerator1, Loop Behavior=::NewEnumerator1, Loop Delay=0.3, UseLoopDelay=true)
   &SpawnBurst_Instantaneous(Spawn Count=3, Use Spawn Probability=false, UseLoopCountLimit=false)
   &SpawnRate(
    SpawnRate=&Multiply_Float(
     A=3
     B=&FloatFromCurve(FloatCurve=DataInterface() {propertyValues {Curve {keys [<at 0 1>,<at 0.4 1>,<at 0.5 0>]}}}, CurveIndex=$Emitter.LoopedAge)
    )
    Use Spawn Probability=false
   )
  ]
  particleSpawnScript [
   &InitializeParticle(
    Color Maximum=<rgba 0.7292 0.9417 1 1>
    Color Minimum=<rgba 0.3759 0.39 0.4271 1>
    Color Mode=::NewEnumerator2
    Lifetime=1.2
    Sprite Rotation Mode=::NewEnumerator2
    Sprite Size Mode=::NewEnumerator1
    Uniform Sprite Size Max=140
    Uniform Sprite Size Min=120
    UsePositionOffset=false
    Write Material Random=true
   )
   &AddVelocity(Velocity=<0 0 180>) {enabled=false}
   &ShapeLocation(
    Handle Radius=10
    Large Radius=35
    MaskSpawnGroup=false
    Offset=<0 0 20>
    Offset Mode=::NewEnumerator0
    Override Seed=false
    Shape Primitive=::NewEnumerator3
    UseOwnerScale=true
   )
   &AddVelocity(Module.Use Falloff Distance=false, Override Seed=false, Velocity Mode=::NewEnumerator1, Velocity Speed=60) {module=AddVelocity001}
   &CameraOffset(Camera Offset Amount=50)
  ]
  particleUpdateScript [
   &ParticleState()
   &Drag(Drag=0.5, Use Linear Drag=true, Use Rotational Drag=true)
   &ScaleColor(
    Scale Alpha=&FloatFromCurve(
     FloatCurve=DataInterface() {propertyValues {Curve {keys [
         {interpMode=RCIM_Cubic, tangentMode=RCTM_User, time=0, value=0}
         {interpMode=RCIM_Cubic, tangentMode=RCTM_User, time=0.5, value=1}
         {tangentMode=RCTM_User, time=1, value=0}
        ]}}}
     Scale Curve=&Multiply_Float(A=1, B=&FNTimeOfDayScalar(Night Scale=0.8, Morning Scale=0.8, Evening Scale=0.8))
    )
    Scale RGB=&MakeVectorFromLinearColorRGB(Color=&FNTimeOfDayColor(Night Color=<rgba 0.4946 0.772 1 1>, Morning Color=<rgba 1 1 1 1>, Day Color=<rgba 1 1 1 1>, Evening Color=<rgba 1 1 1 1>))
    ScaleA=true
    ScaleRGB=true
   )
   &SpriteRotationRate()
   &ScaleSpriteSize(Scale Factor=&Vector2DFromFloat(Value=&FloatFromCurve(FloatCurve=DataInterface() {propertyValues {Curve {keys [{arriveTangent=0.3685, time=0, value=0.8}, <at 0.3 0.9>, {arriveTangent=0.1633, leaveTangent=0.1633, time=1, value=1}]}}})))
   &SubUVAnimation(
    "Sprite Renderer "=DataInterface() {propertyValues {SpriteRenderer=&self:Smoke_Disappear_0.NiagaraSpriteRendererProperties_1}}
    UseEndFrame=false
    UseStartFrame=false
   )
   &DynamicMaterialParameters(
    Index 0 Param 1=1
    Index 0 Param 4=1
    Module.Index0Float0Write=true
    Module.Index0Float1Write=true
    Module.Index0Float2Write=true
    Module.Index0Float3Write=true
   )
   &InheritSourceMovement(Applied Position Scale=<0.5 0.5 0.5>, ApplyForce=false, ApplyPosition=true, ApplyPositionFalloff=false)
   &SolveForcesAndVelocity(Clamp Velocity=false, Limit Acceleration=false)
  ]
  rendererClasses [&NiagaraSpriteRendererProperties]
  renderers [{rendererClass=&NiagaraSpriteRendererProperties}]
 }
}
summary {
 systemName=NS_Emote_MajorSpeech_icicle
 emitters [
  {
   emitterName=Main
   rendererClasses [
    &NiagaraMeshRendererProperties
    &NiagaraSpriteRendererProperties
    &NiagaraLightRendererProperties
   ]
  }
  {
   emitterName=Ice_Shards
   simTarget=GPUComputeSim
   rendererClasses [&NiagaraSpriteRendererProperties, &NiagaraMeshRendererProperties]
  }
  {emitterName=Cloud_Top, rendererClasses [&NiagaraMeshRendererProperties]}
  {emitterName=Smoke_Disappear, rendererClasses [&NiagaraSpriteRendererProperties]}
  {emitterName=Flare, rendererClasses [&NiagaraSpriteRendererProperties]}
  {emitterName=Cloud_Top001, rendererClasses [&NiagaraMeshRendererProperties]}
  {emitterName=Ice, simTarget=GPUComputeSim, rendererClasses [&NiagaraMeshRendererProperties]}
  {emitterName=Ground_MeshSmoke, rendererClasses [&NiagaraMeshRendererProperties]}
  {emitterName=Smoke001, rendererClasses [&NiagaraSpriteRendererProperties]}
  {
   emitterName=Ice_Shards_Top
   simTarget=GPUComputeSim
   rendererClasses [&NiagaraSpriteRendererProperties, &NiagaraMeshRendererProperties]
  }
  {emitterName=Airflow, rendererClasses [&NiagaraMeshRendererProperties]}
 ]
 userVariables [
  User.bIsFrontEnd {type=NiagaraBool}
  User.bIsFrontEndPreview {type=NiagaraBool}
  User.VariantSwitch {type=NiagaraFloat, defaultValue=0, description=""}
 ]
}
